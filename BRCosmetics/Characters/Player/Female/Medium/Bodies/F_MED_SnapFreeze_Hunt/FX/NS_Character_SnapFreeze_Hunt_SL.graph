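# Niagara system for the SnapFreeze Hunt character cosmetic, attaching flame meshes and drifting dot sprites to the body.
# Per-spawn-group arrays choose each particle's skeletal mesh socket, offset and scale; flame materials receive a time-smoothed position.
# User parameters fade effects off, hide selected foot groups and adjust upper-body orientation when skydiving.
# Eye dot, glow and flame emitters, which react to aiming, and a debug emitter are disabled.
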NiagaraSystem /BRCosmetics/Characters/Player/Female/Medium/Bodies/F_MED_SnapFreeze_Hunt/FX/NS_Character_SnapFreeze_Hunt_SL
dependencies {
 usedDataInterfaces [
  &NiagaraDataInterfaceArrayFloat3
  &NiagaraDataInterfaceArrayInt32
  &NiagaraDataInterfaceArrayFloat
  &NiagaraDataInterfaceCurve
  &NiagaraDataInterfaceSkeletalMesh
 ]
 usedDynamicInputs [
  &MergeHelper_ColorSwitch
  &SetBoolByFloatComparison
  &VectorLength
 ]
 usedModules [
  &self:SystemSpawnScript.SystemScriptSource.SystemScriptGraph.NiagaraNodeAssignment_0.SetVariables_#1
  &SystemState
  &self:SystemSpawnScript.SystemScriptSource.SystemScriptGraph.NiagaraNodeAssignment_1.SetVariables_#2
  &FN_Integrator
  &self:SystemSpawnScript.SystemScriptSource.SystemScriptGraph.NiagaraNodeAssignment_2.SetVariables_#3
 ]
}
emitters {
 Dots_Eye {
  bEnabled=false
  emitterUpdateScript [
   &EmitterState(Inactive Response=::NewEnumerator1, Life Cycle Mode=::NewEnumerator1, Loop Delay=0.2, Loop Duration=1, UseLoopDelay=false)
   &SpawnRate(
    Spawn Probability=&Multiply_Float(
     A=&MakeCustomFloatFromBool(Bool=Linked() {linkedVariable {name=User.EyeFlame_On/Off, type=NiagaraBool, defaultValue=~, description=~}}, True Float=1, False Float=0)
     B=&Lerp_Float(A=1, B=0, Alpha=$User.isOff)
    )
    SpawnRate=&Multiply_Float(
     A=&Multiply_Float(A=4, B=&MakeFloatFromInt(Int=$User.EyeFlameAmount))
     B=$System.DotsSpawnRateMult
    )
    Use Spawn Probability=true
   )
  ]
  particleSpawnScript [
   &FN_SetSpawnGroup(Number of Spawn Groups:NiagaraInt32=$User.EyeFlameAmount)
   &InitializeParticle(
    Color=<rgba 0 0 0 1>
    Lifetime=&Multiply_Float(A=&RandomRangeFloat(Minimum=2, Maximum=2.5), B=$System.DotsLifeTimeMult)
    Position Mode=::NewEnumerator0
    Sprite Size Mode=::NewEnumerator1
    Uniform Sprite Size Max=2
    Uniform Sprite Size Min=1.5
    Write Material Random=true
   ) {module=InitializeParticle001}
   &DynamicMaterialParameters(
    Index 0 Param 1=&UniformAOrBFloat(A=0.333, Distribution Weight=0.9, Evaluation Type=::NewEnumerator0, Recalculate Random Each Loop=true, Override Seed=false)
    Module.Index0Param0Write=true
    Module.Index0Param1Write=false
    Module.Index0Param2Write=false
    Module.Index0Param3Write=false
   )
   /SetVariables(Particles.RandomOffset:Vector3f=&Multiply_VectorByFloat(
     Vector=&RandomRangeVector(Minimum=<-1 -1 -1>, Maximum=<1 1 1>)
     Float=&RandomRangeFloat(Minimum=0, Maximum=2)
    )) {isSetParameters=true}
  ]
  particleUpdateScript [
   &SolveForcesAndVelocity(Clamp Velocity=false, Limit Acceleration=false)
   &SkeletalMeshLocation(
    Bone Sampling Mode=::NewEnumerator4
    MaskSpawnGroup=false
    Sampled Position Offset=&Add_Vector(
     A=&Multiply_Vector(
      A=&SelectVectorFromArray(Direct Array Index=$Particles.SpawnGroup, Vector Selection Array=DataInterface() {propertyValues {FloatData [<3.5 6 3>, <3.5 6 -3>]}})
      B=&MakeCustomVectorFromBool(
       Bool=Linked() {linkedVariable {name=User.EyeFlameSide_Left/Right, type=NiagaraBool, defaultValue=~, description=~}}
       True Vector=<1 1 1>
       False Vector=<1 1 -1>
      )
     )
     B=&Add_Vector(
      A=&MakeVector(
       X=&Lerp_Float(
        A=0
        B=10
        Alpha=&FloatFromCurve(FloatCurve=DataInterface() {propertyValues {Curve {keys [<at 0 0 RCIM_Cubic>, {arriveTangent=1.6667, interpMode=RCIM_Cubic, leaveTangent=1.6667, tangentMode=RCTM_User, time=1, value=1}]}}})
       )
       Y=&Lerp_Float(
        A=0
        B=-3
        Alpha=&FloatFromCurve(FloatCurve=DataInterface() {propertyValues {Curve {keys [{tangentMode=RCTM_User, time=0, value=0}, {tangentMode=RCTM_User, time=1, value=1}]}}})
       )
       Z=&Lerp_Float(
        A=0
        B=&Multiply_Float(
         A=&SelectFloatFromArray(Direct Array Index=$Particles.SpawnGroup, Float Selection Array=DataInterface() {propertyValues {FloatData [10,-10]}})
         B=&MakeCustomFloatFromBool(
          Bool=Linked() {linkedVariable {name=User.EyeFlameSide_Left/Right, type=NiagaraBool, defaultValue=~, description=~}}
          True Float=1
          False Float=-1
         )
        )
        Alpha=&FloatFromCurve(FloatCurve=DataInterface() {propertyValues {Curve {keys [{tangentMode=RCTM_User, time=0, value=0}, {tangentMode=RCTM_User, time=1, value=1}]}}})
       )
      )
      B=$Particles.RandomOffset
     )
    )
    Skeletal Mesh=$System.Skeletal Mesh
   )
   &ParticleState() {module=ParticleState001}
   &ParticleState()
   &CameraOffset(Camera Offset Amount=5)
   &ScaleColor(
    Scale Alpha=&Multiply_Float(
     A=&MakeCustomFloatFromBool(Bool=&Boolean_OrOperation(Bool1=$User.bisADS, Bool2=$User.bIsADSOther), True Float=0, False Float=1)
     B=1
    )
    ScaleA=true
    ScaleRGB=false
   ) {module=ScaleColor001}
   /SetVariables(Particles.SpriteSize=&Multiply_Vector2D_ByFloat(
     Float=&Multiply_Float(
      A=1
      B=&UniformAOrBFloat(
       A=&FloatFromCurve(FloatCurve=DataInterface() {propertyValues {Curve {keys [
            <at 0 0 RCIM_Cubic>
            {arriveTangent=9.6335, interpMode=RCIM_Cubic, leaveTangent=2.7827, tangentMode=RCTM_Break, time=0.15, value=1}
            <at 0.4 0.6 RCIM_Cubic>
            {arriveTangent=2.7202, interpMode=RCIM_Cubic, leaveTangent=-2.3793, tangentMode=RCTM_Break, time=0.7, value=1}
            <at 1 0 RCIM_Cubic>
           ]}}})
       B=&FloatFromCurve(FloatCurve=DataInterface() {propertyValues {Curve {keys [
            <at 0 0 RCIM_Cubic>
            {arriveTangent=9.6335, interpMode=RCIM_Cubic, leaveTangent=2.7827, tangentMode=RCTM_Break, time=0.15, value=0.9794}
            {arriveTangent=0.0457, interpMode=RCIM_Cubic, leaveTangent=0.0457, time=0.5397, value=0.5}
            {arriveTangent=2.7202, interpMode=RCIM_Cubic, leaveTangent=-2.3793, tangentMode=RCTM_Break, time=0.6, value=1}
            <at 1 0 RCIM_Cubic>
           ]}}})
       Evaluation Type=::NewEnumerator0
       Recalculate Random Each Loop=true
       Override Seed=false
      )
     )
     Vector2D=$Particles.Initial.SpriteSize
    )) {isSetParameters=true}
   /SetVariables(
    Particles.AdditionalDot01Position:NiagaraPosition=&AddVectorToPosition(
     Position=$Particles.Position
     Vector=&Multiply_VectorByFloat(
      Vector=&RandomRangeVector(Minimum=<-1 -1 -1>, Maximum=<1 1 1>)
      Float=&Waveform(
       Global Input Position=&RandomRangeFloat(Minimum=0, Maximum=1)
       "Waveform [1]"=::NewEnumerator0
       "[1] Frequency"=1
       "[1] Amplitude Min/Max"=<0 1.5>
      )
     )
    )
    Particles.AdditionalDot01SpriteSize:Vector2f=&Multiply_Vector2D_ByFloat(
     Float=&Multiply_Float(
      A=&UniformAOrBFloat(
       A=&FloatFromCurve(FloatCurve=DataInterface() {propertyValues {Curve {keys [
            <at 0 0 RCIM_Cubic>
            {arriveTangent=9.6335, interpMode=RCIM_Cubic, leaveTangent=2.7827, tangentMode=RCTM_Break, time=0.15, value=1}
            <at 0.6943 0.7273 RCIM_Cubic>
            {arriveTangent=2.7202, interpMode=RCIM_Cubic, leaveTangent=-2.3793, tangentMode=RCTM_Break, time=0.8055, value=1}
            <at 0.95 0 RCIM_Cubic>
           ]}}})
       B=&FloatFromCurve(FloatCurve=DataInterface() {propertyValues {Curve {keys [
            <at 0 0 RCIM_Cubic>
            {arriveTangent=9.6335, interpMode=RCIM_Cubic, leaveTangent=2.7827, tangentMode=RCTM_Break, time=0.15, value=0.9794}
            {arriveTangent=0.0322, interpMode=RCIM_Cubic, leaveTangent=0.0322, time=0.5397, value=0.5}
            {arriveTangent=2.7202, interpMode=RCIM_Cubic, leaveTangent=-2.3793, tangentMode=RCTM_Break, time=0.7888, value=1}
            <at 0.95 0 RCIM_Cubic>
           ]}}})
       Evaluation Type=::NewEnumerator0
       Recalculate Random Each Loop=true
       Override Seed=false
      )
      B=1
     )
     Vector2D=&Vector2DFromFloat(Value=&RandomRangeFloat(Minimum=0.3, Maximum=1.2))
    )
    Particles.AdditionalDot01Visibility:NiagaraInt32=&MakeIntFromFloat(Float=&UniformAOrBFloat(A=0, Distribution Weight=0.6, Evaluation Type=::NewEnumerator0, Recalculate Random Each Loop=true, Override Seed=false))
   ) {isSetParameters=true}
   /SetVariables(
    Particles.AdditionalDot02Position:NiagaraPosition=&AddVectorToPosition(
     Position=$Particles.Position
     Vector=&Multiply_VectorByFloat(
      Vector=&RandomRangeVector(Minimum=<-1 -1 -1>, Maximum=<1 1 1>)
      Float=&Waveform(
       Global Input Position=&RandomRangeFloat(Minimum=0, Maximum=1)
       "Waveform [1]"=::NewEnumerator0
       "[1] Frequency"=1
       "[1] Amplitude Min/Max"=<0.5 1>
      )
     )
    )
    Particles.AdditionalDot02SpriteSize:Vector2f=&Multiply_Vector2D_ByFloat(
     Float=&Multiply_Float(
      A=&UniformAOrBFloat(
       A=&FloatFromCurve(FloatCurve=DataInterface() {propertyValues {Curve {keys [
            <at 0 0 RCIM_Cubic>
            {arriveTangent=9.6335, interpMode=RCIM_Cubic, leaveTangent=2.7827, tangentMode=RCTM_Break, time=0.15, value=1}
            <at 0.6943 0.7273 RCIM_Cubic>
            {arriveTangent=2.7202, interpMode=RCIM_Cubic, leaveTangent=-2.3793, tangentMode=RCTM_Break, time=0.8055, value=1}
            <at 0.95 0 RCIM_Cubic>
           ]}}})
       B=&FloatFromCurve(FloatCurve=DataInterface() {propertyValues {Curve {keys [
            <at 0 0 RCIM_Cubic>
            {arriveTangent=9.6335, interpMode=RCIM_Cubic, leaveTangent=2.7827, tangentMode=RCTM_Break, time=0.15, value=0.9794}
            {arriveTangent=0.0322, interpMode=RCIM_Cubic, leaveTangent=0.0322, time=0.5397, value=0.5}
            {arriveTangent=2.7202, interpMode=RCIM_Cubic, leaveTangent=-2.3793, tangentMode=RCTM_Break, time=0.7888, value=1}
            <at 0.95 0 RCIM_Cubic>
           ]}}})
       Evaluation Type=::NewEnumerator0
       Recalculate Random Each Loop=true
       Override Seed=false
      )
      B=1
     )
     Vector2D=&Vector2DFromFloat(Value=&RandomRangeFloat(Minimum=0.5, Maximum=1))
    )
    Particles.AdditionalDot02Visibility:NiagaraInt32=&MakeIntFromFloat(Float=&UniformAOrBFloat(A=0, Distribution Weight=0.88, Evaluation Type=::NewEnumerator0, Recalculate Random Each Loop=true, Override Seed=false))
   ) {isSetParameters=true}
  ]
  rendererClasses [&NiagaraSpriteRendererProperties]
  renderers [
   {rendererClass=&NiagaraSpriteRendererProperties}
   {rendererClass=&NiagaraSpriteRendererProperties, rendererIndex=1}
   {rendererClass=&NiagaraSpriteRendererProperties, rendererIndex=2}
  ]
 }
 EyeGlow {
  bEnabled=false
  emitterUpdateScript [
   &EmitterState(
    Inactive Response=::NewEnumerator1
    Life Cycle Mode=::NewEnumerator1
    Loop Behavior=::NewEnumerator1
    Loop Duration Mode=::NewEnumerator1
    UseLoopDelay=false
   )
   &SpawnBurst_Instantaneous(Spawn Count=2, Use Spawn Probability=true, UseLoopCountLimit=true)
  ]
  particleSpawnScript [
   &FN_SetSpawnGroup(Number of Spawn Groups=2)
   &InitializeParticle(Color Mode=::NewEnumerator0, UsePositionOffset=false, Write Material Random=true)
  ]
  particleUpdateScript [
   &ParticleState(Kill Particles When Lifetime Has Elapsed=false, Let Infinitely Lived Particles Die When Emitter Deactivates=true)
   /SetVariables(Particles.MainColor:LinearColor=&MultiplyLinearColor(
     Linear Color=&MultiplyLinearColor(
      Linear Color=$System.ColorA
      Scale Factor=&MakeVector4(
       X=1
       Y=1
       Z=1
       W=&Multiply_Float(A=0.2, B=&Lerp_Float(A=1, B=0, Alpha=$User.isOff))
      )
     )
     Scale Factor=&MakeVector4_FromFloat(Float=&Multiply_Float(
       A=&MaskFloatBySpawnGroup(
        Masked Float=&MakeCustomFloatFromBool(Bool=&Boolean_OrOperation(Bool1=$User.bisADS, Bool2=$User.bIsADSOther), True Float=0, False Float=1)
        Passthrough Float=1
       )
       B=&Waveform(
        Waveform Count=::NewEnumerator1
        Global Input Position=&ReturnNormalizedExecIndex(Normalized Index Scale=9)
        "Waveform [1]"=::NewEnumerator12
        "[1] Frequency"=2.65
        "[1] Amplitude Min/Max"=<0.7 1>
        "Waveform [2] Blend"=::NewEnumerator5
        "[2] Interpolation"=&FloatFromCurve(FloatCurve=DataInterface() {propertyValues {Curve {keys [
             {tangentMode=RCTM_User, time=0, value=0}
             <at 0.1 1>
             {arriveTangent=-0.0405, leaveTangent=-0.0405, time=0.9, value=1}
             {arriveTangent=-8.0089, interpMode=RCIM_Cubic, leaveTangent=-8.0089, tangentMode=RCTM_User, time=1, value=0}
            ]}}})
        "[2] Invert Interpolation"=false
        "Waveform [2]"=::NewEnumerator12
        "[2] Input Position"=0
        "[2] Phase"=$Emitter.Age
        "[2] Phase Scale"=1
        "[2] Compound Frequency"=2
        "[2] Frequency"=16
        "[2] Amplitude Min/Max"=<0.75 1>
        "[2] Scale"=1
        "[2] Bias"=0
        "[2] Exponent"=1
        "[2] Pulse Duration"=0.1
        "[2] Clip Triangle Min/Max"=<0 0>
        "[2] Random Seed"=0
       )
      ))
    )) {isSetParameters=true}
   /SetVariables(
    Particles.SoftGlowCameraOffset:NiagaraFloat=&SelectFloatFromArray(Direct Array Index=$Particles.SpawnGroup, Float Selection Array=DataInterface() {propertyValues {FloatData [1,1]}})
    Particles.SoftGlowColor:LinearColor=&MultiplyLinearColor(
     Linear Color=$StackContext.MainColor
     Scale Factor=&MakeFloat4FromVectorAndFloat(
      "Vector (RGB)"=&VectorFromFloat(Value=&SelectFloatFromArray(Direct Array Index=$Particles.SpawnGroup, Float Selection Array=DataInterface() {propertyValues {FloatData [3,3]}}))
      "Float (Alpha)"=0.5
     )
    )
    Particles.SoftGlowSize:Vector2f=&Vector2DFromFloat(Value=&SelectFloatFromArray(Direct Array Index=$Particles.SpawnGroup, Float Selection Array=DataInterface() {propertyValues {FloatData [9,9]}}))
   ) {isSetParameters=true}
   /SetVariables(
    Particles.FakeLightColor:LinearColor=&MultiplyLinearColor(
     Linear Color=$StackContext.MainColor
     Scale Factor=&MakeFloat4FromVectorAndFloat(
      "Vector (RGB)"=&VectorFromFloat(Value=&SelectFloatFromArray(Direct Array Index=$Particles.SpawnGroup, Float Selection Array=DataInterface() {propertyValues {FloatData [8,8]}}))
      "Float (Alpha)"=0.25
     )
    )
    Particles.FakeLightSize:Vector2f=&Vector2DFromFloat(Value=&SelectFloatFromArray(Direct Array Index=$Particles.SpawnGroup, Float Selection Array=DataInterface() {propertyValues {FloatData [8,8]}}))
   ) {isSetParameters=true}
   /SetVariables(
    Particles.CapsuleGlowColor:LinearColor=&MultiplyLinearColor(
     Linear Color=$StackContext.MainColor
     Scale Factor=&MakeFloat4FromVectorAndFloat(
      "Vector (RGB)"=&VectorFromFloat(Value=&SelectFloatFromArray(Direct Array Index=$Particles.SpawnGroup, Float Selection Array=DataInterface() {propertyValues {FloatData [3,3]}}))
      "Float (Alpha)"=1
     )
    )
    Particles.CapsuleSpriteSize:Vector2f=&Vector2DFromFloat(Value=&SelectFloatFromArray(Direct Array Index=$Particles.SpawnGroup, Float Selection Array=DataInterface() {propertyValues {FloatData [3,3]}}))
   ) {isSetParameters=true}
   /SetVariables(
    Particles.LightColor=&MultiplyLinearColor(
     Linear Color=$StackContext.MainColor
     Scale Factor=&MakeFloat4FromVectorAndFloat(
      "Vector (RGB)"=&VectorFromFloat(Value=&SelectFloatFromArray(Direct Array Index=$Particles.SpawnGroup, Float Selection Array=DataInterface() {propertyValues {FloatData [1,1]}}))
      "Float (Alpha)"=33
     )
    )
    Particles.LightRadius=&SelectFloatFromArray(Direct Array Index=$Particles.SpawnGroup, Float Selection Array=DataInterface() {propertyValues {FloatData [33,33]}})
   ) {isSetParameters=true}
   &SkeletalMeshLocation(
    Bone / Socket Index=&SelectIntFromArray(Direct Array Index=$Particles.SpawnGroup, Int Selection Array=DataInterface() {propertyValues {IntData [0,0]}})
    Bone Sampling Mode=::NewEnumerator4
    MaskSpawnGroup=false
    Sampled Position Offset=&SelectVectorFromArray(Direct Array Index=$Particles.SpawnGroup, Vector Selection Array=DataInterface() {propertyValues {FloatData [<3.5 6.3 -3>, <3.5 6.3 3>]}})
    Skeletal Mesh=$System.Skeletal Mesh
   )
   /SetVariables(Particles.CapsuleSpriteAlignment:Vector3f=&RotateVectorByQuaternion(
     Vector To Rotate=&SelectVectorFromArray(Direct Array Index=$Particles.SpawnGroup, Vector Selection Array=DataInterface() {propertyValues {FloatData [<0 0.25 1>, <0 -0.25 1>]}})
     Quaternion=$Particles.SkeletalMeshLocation.SampledRotation
    )) {isSetParameters=true}
   /SetVariables(Particles.CapsuleSpriteDynamic:Vector4f=&MakeFloat4FromVectorAndFloat(
     "Vector (RGB)"=&NormalizeVector(Vector To Normalize=$Particles.CapsuleSpriteAlignment, Safe Normalize=true, Safe Fallback Threshold=0, Safe Fallback Vector=<0 0 0.001>)
     "Float (Alpha)"=&SelectFloatFromArray(Direct Array Index=$Particles.SpawnGroup, Float Selection Array=DataInterface() {propertyValues {FloatData [5,5]}})
    )) {isSetParameters=true}
   /SetVariables(
    Particles.EyeOverlayColor:LinearColor=&MultiplyLinearColor(
     Linear Color=$System.ColorEye
     Scale Factor=&MakeVector4(X=1, Y=1, Z=1, W=&Lerp_Float(A=1, B=0, Alpha=$User.isOff))
    )
    Particles.EyeOverlayPosition:NiagaraPosition=$Output.SkeletalMeshLocation.SampledPositionWithoutOffset
    Particles.EyeOverlayScale:Vector3f=&SelectVectorFromArray(Direct Array Index=$Particles.SpawnGroup, Vector Selection Array=DataInterface() {propertyValues {FloatData [<1 1 1>, <1 1 -1>]}})
   ) {isSetParameters=true}
   &ScaleColor(Scale Alpha=&Lerp_Float(A=1, B=0, Alpha=$User.isOff), ScaleA=true, ScaleRGB=false) {module=ScaleColor002}
  ]
  rendererClasses [&NiagaraSpriteRendererProperties]
  renderers [
   {rendererClass=&NiagaraSpriteRendererProperties}
   {rendererClass=&NiagaraSpriteRendererProperties, rendererIndex=1}
   {rendererClass=&NiagaraSpriteRendererProperties, rendererIndex=2}
  ]
 }
 Eye_Flame {
  bEnabled=false
  emitterUpdateScript [
   &EmitterState(
    Inactive Response=::NewEnumerator1
    Life Cycle Mode=::NewEnumerator1
    Loop Behavior=::NewEnumerator1
    Loop Duration Mode=::NewEnumerator1
    Scalability Mode=::NewEnumerator1
    UseLoopDelay=false
   )
   &SpawnBurst_Instantaneous(
    Spawn Count=$User.EyeFlameAmount
    Spawn Probability=&Multiply_Float(
     A=&MakeCustomFloatFromBool(Bool=Linked() {linkedVariable {name=User.EyeFlame_On/Off, type=NiagaraBool, defaultValue=~, description=~}}, True Float=1, False Float=0)
     B=1
    )
    Use Spawn Probability=true
    UseLoopCountLimit=false
   )
  ]
  particleSpawnScript [
   &FN_SetSpawnGroup(Number of Spawn Groups:NiagaraInt32=$User.EyeFlameAmount)
   &InitializeParticle(
    Color=<rgba 1 1 1 0.9>
    Lifetime=5
    Mesh Scale=&Multiply_Vector(
     A=&SelectVectorFromArray(Direct Array Index=$Particles.SpawnGroup, Vector Selection Array=DataInterface() {propertyValues {FloatData [<1 1 1>, <1 1 -1>]}})
     B=&MakeCustomVectorFromBool(
      Bool=Linked() {linkedVariable {name=User.EyeFlameSide_Left/Right, type=NiagaraBool, defaultValue=~, description=~}}
      True Vector=<1 1 1>
      False Vector=<1 1 -1>
     )
    )
    Mesh Scale Mode=::NewEnumerator4
    UsePositionOffset=false
    Write Material Random=true
   )
  ]
  particleUpdateScript [
   &ParticleState(Kill Particles When Lifetime Has Elapsed=false)
   &SkeletalMeshLocation(
    Bone Sampling Mode=::NewEnumerator4
    MaskSpawnGroup=false
    Sampled Position Offset=&Add_Vector(
     A=&Multiply_Vector(
      A=<-1 -1 0.5>
      B=&MakeCustomVectorFromBool(
       Bool=Linked() {linkedVariable {name=User.EyeFlameSide_Left/Right, type=NiagaraBool, defaultValue=~, description=~}}
       True Vector=<1 1 1>
       False Vector=<1 1 -1>
      )
     )
     B=&FN_FrontEndVector3Modifier(Gameplay=<0 0 0>, Front End=<-0.2 0.4 0.5>)
    )
    Skeletal Mesh=$System.Skeletal Mesh
   )
   &InterpolateOverTime(Rate Of Change=10, Target Vector Value:Vector3f=$Particles.SkeletalMeshLocation.SampledPosition, Variable Type=::NewEnumerator2)
   &DynamicMaterialParameters(
    Index 0 Param 1=1
    Index 0 Param 2=1
    Index 0 Param 3=$System.FN_Integrator002.Value
    Index 1 Param 1=5
    Index 1 Param 3=1.5
    Index 2 Param 1=&MakeFloatFromVector(Vector=$StackContext.InterpolateOverTime.InterpolateOverTime_V3_.Moving Average, Channel=::NewEnumerator0)
    Index 2 Param 2=&MakeFloatFromVector(Vector=$StackContext.InterpolateOverTime.InterpolateOverTime_V3_.Moving Average, Channel=::NewEnumerator1)
    Index 2 Param 3=&MakeFloatFromVector(Vector=$StackContext.InterpolateOverTime.InterpolateOverTime_V3_.Moving Average, Channel=::NewEnumerator2)
    Index 2 Param 4=1.5
    Module.Index0Param0Write=true
    Module.Index0Param1Write=true
    Module.Index0Param2Write=true
    Module.Index0Param3Write=false
    Module.Index1Param0Write=true
    Module.Index1Param1Write=true
    Module.Index1Param2Write=true
    Module.Index1Param3Write=false
    Module.Index2Param0Write=true
    Module.Index2Param1Write=true
    Module.Index2Param2Write=true
    Module.Index2Param3Write=false
    Write Parameter Index 1=true
    Write Parameter Index 2=true
   )
   &ScaleColor(
    Scale Alpha=&Multiply_Float(
     A=0.5
     B=&Lerp_Float(
      A=0
      B=1
      Alpha=&MakeCustomFloatFromBool(Bool=&Boolean_OrOperation(Bool1=$User.bisADS, Bool2=$User.bIsADSOther), True Float=0, False Float=1)
     )
    )
    ScaleA=true
    ScaleRGB=false
   )
   &ScaleColor(Scale Alpha=&Lerp_Float(A=1, B=0, Alpha=$User.isOff), ScaleA=true, ScaleRGB=false) {module=ScaleColor002}
  ]
  rendererClasses [&NiagaraMeshRendererProperties]
  renderers [{rendererClass=&NiagaraMeshRendererProperties, rendererIndex=0}]
 }
 FN_Debug {
  bEnabled=false
  emitterUpdateScript [
   &EmitterState()
   &SpawnBurst_Instantaneous(Spawn Count=1, Use Spawn Probability=false)
  ]
  particleSpawnScript [/Niagara/Modules/Spawn/Initialization/InitializeParticle(
    Lifetime=5
    Module.Write Color=true
    Module.Write Lifetime=true
    Module.Write Mass=true
    Module.Write Material Random=true
    Module.Write Position=false
    Module.Write Scale=false
    Module.Write SpriteRotation=false
    Module.Write SpriteSize=true
    Position=&SimulationPosition()
    Sprite Size=<100 100>
   )]
  particleUpdateScript [
   &ParticleState()
   /SetVariables(Particles.Debug=0.1) {isSetParameters=true}
   &DynamicMaterialParameters(
    Index 0 Param 1=&MakeCustomFloatFromBool(Bool=$User.bisFootHidden, True Float=1, False Float=0)
    Module.Param0WriteEnabled=true
    Module.Param1WriteEnabled=true
    Module.Param2WriteEnabled=true
    Module.Param3WriteEnabled=true
   )
   &SystemLocation(MaskSpawnGroup=false, Offset=<0 0 200>)
  ]
  rendererClasses [&NiagaraSpriteRendererProperties]
  renderers [{rendererClass=&NiagaraSpriteRendererProperties}]
 }
 LowerBody_Dots {
  emitterUpdateScript [
   &EmitterState(Inactive Response=::NewEnumerator1, Life Cycle Mode=::NewEnumerator1, Loop Delay=0.2, Loop Duration=1, UseLoopDelay=false)
   &SpawnRate(
    Spawn Probability=&Multiply_Float(A=0.9, B=&Lerp_Float(A=1, B=0, Alpha=$User.isOff))
    SpawnRate=&Multiply_Float(A=26, B=$System.DotsSpawnRateMult)
    Use Spawn Probability=true
   )
  ]
  particleSpawnScript [
   &FN_SetSpawnGroup(Number of Spawn Groups=14)
   /Niagara/Modules/Spawn/Initialization/V2/InitializeParticle(
    Color=<rgba 0 0 0 1>
    Lifetime=&Multiply_Float(A=&RandomRangeFloat(Minimum=1.5, Maximum=2), B=$System.DotsLifeTimeMult)
    Position Mode=::NewEnumerator0
    Sprite Size Mode=::NewEnumerator3
    Uniform Sprite Size=&RandomRangeFloat(Minimum=1.5, Maximum=2)
    Write Material Random=true
   ) {module=InitializeParticle001}
   &DynamicMaterialParameters(
    Index 0 Param 1=&UniformAOrBFloat(A=0.333, Distribution Weight=0.9, Evaluation Type=::NewEnumerator0, Recalculate Random Each Loop=true, Override Seed=false)
    Module.Index0Param0Write=true
    Module.Index0Param1Write=false
    Module.Index0Param2Write=false
    Module.Index0Param3Write=false
   )
   /SetVariables(Particles.RandomOffset:Vector3f=&Multiply_VectorByFloat(
     Vector=&RandomRangeVector(Minimum=<-1 -1 -1>, Maximum=<1 1 2>)
     Float=&RandomRangeFloat(Minimum=0, Maximum=3)
    )) {isSetParameters=true}
  ]
  particleUpdateScript [
   &ParticleState() {module=ParticleState001}
   &SkeletalMeshLocation(
    Bone / Socket Index=&SelectIntFromArray(Direct Array Index=$Particles.SpawnGroup, Int Selection Array=DataInterface() {propertyValues {IntData [0,1,2,3,4,5,6,7,8,22,23,30,31,32]}})
    MaskSpawnGroup=false
    Mesh Sampling Type=::NewEnumerator5
    Orientation Sampling=::NewEnumerator1
    Sampled Position Offset=&MakeVector(
     X=&Lerp_Float(
      A=0
      B=&Multiply_Float(
       A=&RandomRangeFloat(Minimum=0, Maximum=2)
       B=&SelectFloatFromArray(
        Direct Array Index=$Particles.SpawnGroup
        Float Selection Array=DataInterface() {propertyValues {FloatData [2,2,-2,-2,-2,-4,-1,-3,-2,-1,-2,-1,-2,-2]}}
       )
      )
      Alpha=&FloatFromCurve(FloatCurve=DataInterface() {propertyValues {Curve {keys [
           {interpMode=RCIM_Cubic, tangentMode=RCTM_Break, time=0, value=0}
           {arriveTangent=1.4797, interpMode=RCIM_Cubic, tangentMode=RCTM_Break, time=1, value=1}
          ]}}})
     )
     Y=0
     Z=&Lerp_Float(
      A=0
      B=&Multiply_Float(
       A=&SelectFloatFromArray(
        Direct Array Index=$Particles.SpawnGroup
        Float Selection Array=DataInterface() {propertyValues {FloatData [0.5,1,0.7,1,0.7,1.1,0.5,0.6,1.1,0.3,1,0.6,0.6,0.3]}}
       )
       B=&RandomRangeFloat(Minimum=8, Maximum=10)
      )
      Alpha=&FloatFromCurve(FloatCurve=DataInterface() {propertyValues {Curve {keys [<at 0 0>,<at 1 1>]}}})
     )
    )
    Skeletal Mesh=$System.Skeletal Mesh
    Socket Sampling Mode=::NewEnumerator5
   )
   /SetVariables(Particles.Position=&AddVectorToPosition(Position=$Particles.Position, Vector=$Particles.RandomOffset)) {isSetParameters=true}
   &ParticleState()
   &CameraOffset(Camera Offset Amount=5)
   &ScaleColor(
    Scale Alpha=&MakeCustomFloatFromBool(
     Bool=$User.bisFootHidden
     True Float=&SelectFloatFromArray(Direct Array Index=$Particles.SpawnGroup, Float Selection Array=DataInterface() {propertyValues {FloatData [1,1,1,1,0,0,0,0,1,0,0,1,1,0]}})
     False Float=1
    )
    ScaleA=true
    ScaleRGB=false
   ) {module=ScaleColor001}
   /SetVariables(Particles.SpriteSize=&Multiply_Vector2D_ByFloat(
     Float=&Multiply_Float(
      A=1
      B=&UniformAOrBFloat(
       A=&FloatFromCurve(FloatCurve=DataInterface() {propertyValues {Curve {keys [
            <at 0 0 RCIM_Cubic>
            {arriveTangent=9.6335, interpMode=RCIM_Cubic, leaveTangent=2.7827, tangentMode=RCTM_Break, time=0.15, value=1}
            <at 0.4 0.6 RCIM_Cubic>
            {arriveTangent=2.7202, interpMode=RCIM_Cubic, leaveTangent=-2.3793, tangentMode=RCTM_Break, time=0.7, value=1}
            <at 1 0 RCIM_Cubic>
           ]}}})
       B=&FloatFromCurve(FloatCurve=DataInterface() {propertyValues {Curve {keys [
            <at 0 0 RCIM_Cubic>
            {arriveTangent=9.6335, interpMode=RCIM_Cubic, leaveTangent=2.7827, tangentMode=RCTM_Break, time=0.15, value=0.9794}
            {arriveTangent=0.0457, interpMode=RCIM_Cubic, leaveTangent=0.0457, time=0.5397, value=0.5}
            {arriveTangent=2.7202, interpMode=RCIM_Cubic, leaveTangent=-2.3793, tangentMode=RCTM_Break, time=0.6, value=1}
            <at 1 0 RCIM_Cubic>
           ]}}})
       Evaluation Type=::NewEnumerator0
       Recalculate Random Each Loop=true
       Override Seed=false
      )
     )
     Vector2D=$Particles.Initial.SpriteSize
    )) {isSetParameters=true}
   /SetVariables(
    Particles.AdditionalDot01Position:NiagaraPosition=&AddVectorToPosition(
     Position=$Particles.Position
     Vector=&Multiply_VectorByFloat(
      Vector=&RandomRangeVector(Minimum=<-1 -1 -1>, Maximum=<1 1 1>)
      Float=&Waveform(
       Global Input Position=&RandomRangeFloat(Minimum=0, Maximum=1)
       "Waveform [1]"=::NewEnumerator0
       "[1] Frequency"=1
       "[1] Amplitude Min/Max"=<0 1.5>
      )
     )
    )
    Particles.AdditionalDot01SpriteSize:Vector2f=&Multiply_Vector2D_ByFloat(
     Float=&Multiply_Float(
      A=&UniformAOrBFloat(
       A=&FloatFromCurve(FloatCurve=DataInterface() {propertyValues {Curve {keys [
            <at 0 0 RCIM_Cubic>
            {arriveTangent=9.6335, interpMode=RCIM_Cubic, leaveTangent=2.7827, tangentMode=RCTM_Break, time=0.15, value=1}
            <at 0.6943 0.7273 RCIM_Cubic>
            {arriveTangent=2.7202, interpMode=RCIM_Cubic, leaveTangent=-2.3793, tangentMode=RCTM_Break, time=0.8055, value=1}
            <at 0.95 0 RCIM_Cubic>
           ]}}})
       B=&FloatFromCurve(FloatCurve=DataInterface() {propertyValues {Curve {keys [
            <at 0 0 RCIM_Cubic>
            {arriveTangent=9.6335, interpMode=RCIM_Cubic, leaveTangent=2.7827, tangentMode=RCTM_Break, time=0.15, value=0.9794}
            {arriveTangent=0.0322, interpMode=RCIM_Cubic, leaveTangent=0.0322, time=0.5397, value=0.5}
            {arriveTangent=2.7202, interpMode=RCIM_Cubic, leaveTangent=-2.3793, tangentMode=RCTM_Break, time=0.7888, value=1}
            <at 0.95 0 RCIM_Cubic>
           ]}}})
       Evaluation Type=::NewEnumerator0
       Recalculate Random Each Loop=true
       Override Seed=false
      )
      B=1
     )
     Vector2D=&Vector2DFromFloat(Value=&RandomRangeFloat(Minimum=1, Maximum=1.5))
    )
    Particles.AdditionalDot01Visibility:NiagaraInt32=&MakeIntFromFloat(Float=&UniformAOrBFloat(A=0, Distribution Weight=0.6, Evaluation Type=::NewEnumerator0, Recalculate Random Each Loop=true, Override Seed=false))
   ) {isSetParameters=true}
   /SetVariables(
    Particles.AdditionalDot02Position:NiagaraPosition=&AddVectorToPosition(
     Position=$Particles.Position
     Vector=&Multiply_VectorByFloat(
      Vector=&RandomRangeVector(Minimum=<-1 -1 -1>, Maximum=<1 1 1>)
      Float=&Waveform(
       Global Input Position=&RandomRangeFloat(Minimum=0, Maximum=1)
       "Waveform [1]"=::NewEnumerator0
       "[1] Frequency"=1
       "[1] Amplitude Min/Max"=<0 2>
      )
     )
    )
    Particles.AdditionalDot02SpriteSize:Vector2f=&Multiply_Vector2D_ByFloat(
     Float=&Multiply_Float(
      A=&UniformAOrBFloat(
       A=&FloatFromCurve(FloatCurve=DataInterface() {propertyValues {Curve {keys [
            <at 0 0 RCIM_Cubic>
            {arriveTangent=9.6335, interpMode=RCIM_Cubic, leaveTangent=2.7827, tangentMode=RCTM_Break, time=0.15, value=1}
            <at 0.6943 0.7273 RCIM_Cubic>
            {arriveTangent=2.7202, interpMode=RCIM_Cubic, leaveTangent=-2.3793, tangentMode=RCTM_Break, time=0.8055, value=1}
            <at 0.95 0 RCIM_Cubic>
           ]}}})
       B=&FloatFromCurve(FloatCurve=DataInterface() {propertyValues {Curve {keys [
            <at 0 0 RCIM_Cubic>
            {arriveTangent=9.6335, interpMode=RCIM_Cubic, leaveTangent=2.7827, tangentMode=RCTM_Break, time=0.15, value=0.9794}
            {arriveTangent=0.0322, interpMode=RCIM_Cubic, leaveTangent=0.0322, time=0.5397, value=0.5}
            {arriveTangent=2.7202, interpMode=RCIM_Cubic, leaveTangent=-2.3793, tangentMode=RCTM_Break, time=0.7888, value=1}
            <at 0.95 0 RCIM_Cubic>
           ]}}})
       Evaluation Type=::NewEnumerator0
       Recalculate Random Each Loop=true
       Override Seed=false
      )
      B=1
     )
     Vector2D=&Vector2DFromFloat(Value=&RandomRangeFloat(Minimum=1.5, Maximum=2))
    )
    Particles.AdditionalDot02Visibility:NiagaraInt32=&MakeIntFromFloat(Float=&UniformAOrBFloat(A=0, Distribution Weight=0.88, Evaluation Type=::NewEnumerator0, Recalculate Random Each Loop=true, Override Seed=false))
   ) {isSetParameters=true}
   &ScaleColor(
    Scale Alpha=&Multiply_Float(A=&RandomRangeFloat(Minimum=0.5, Maximum=0.8), B=1)
    Scale RGB=<1 0 0>
    ScaleA=false
    ScaleRGB=false
   )
  ]
  rendererClasses [&NiagaraSpriteRendererProperties]
  renderers [
   {rendererClass=&NiagaraSpriteRendererProperties}
   {rendererClass=&NiagaraSpriteRendererProperties, rendererIndex=1}
   {rendererClass=&NiagaraSpriteRendererProperties, rendererIndex=2}
  ]
 }
 LowerBody_Flame {
  emitterUpdateScript [
   &EmitterState(
    Life Cycle Mode=::NewEnumerator1
    Loop Behavior=::NewEnumerator1
    Loop Duration Mode=::NewEnumerator1
    Scalability Mode=::NewEnumerator1
    UseLoopDelay=false
   )
   &SpawnBurst_Instantaneous(Spawn Count=14, Use Spawn Probability=false, UseLoopCountLimit=false)
  ]
  particleSpawnScript [
   &FN_SetSpawnGroup(Number of Spawn Groups=14)
   &InitializeParticle(
    Lifetime=5
    Mesh Scale=&MakeVector(
     X=1
     Y=1
     Z=&Multiply_Float(
      A=&Multiply_Float(A=0.68, B=0.72)
      B=&SelectFloatFromArray(
       Direct Array Index=$Particles.SpawnGroup
       Float Selection Array=DataInterface() {propertyValues {FloatData [0.08,0.15,0.18,0.08,0.15,0.15,0.08,0.15,0.2,0.12,0.2,0.1,0.1,0.13]}}
      )
     )
    )
    Mesh Scale Mode=::NewEnumerator4
    UsePositionOffset=false
    Write Material Random=true
   )
  ]
  particleUpdateScript [
   &ParticleState(Kill Particles When Lifetime Has Elapsed=false)
   &SkeletalMeshLocation(
    Bone / Socket Index=&SelectIntFromArray(Direct Array Index=$Particles.SpawnGroup, Int Selection Array=DataInterface() {propertyValues {IntData [0,1,2,3,4,5,6,7,8,22,23,30,31,32]}})
    MaskSpawnGroup=false
    Mesh Sampling Type=::NewEnumerator5
    Sampled Position Offset Coordinate Space:ENiagaraSkelMeshTransforms=::NewEnumerator2
    Skeletal Mesh=$System.Skeletal Mesh
    Socket Sampling Mode=::NewEnumerator5
   )
   &InterpolateOverTime(Rate Of Change=10, Target Vector Value:Vector3f=$Particles.SkeletalMeshLocation.SampledPosition, Variable Type=::NewEnumerator2)
   &DynamicMaterialParameters(
    Index 0 Param 1=&Multiply_Float(
     A=0.7
     B=&SelectFloatFromArray(
      Direct Array Index=$Particles.SpawnGroup
      Float Selection Array=DataInterface() {propertyValues {FloatData [0.5,0.5,0.5,0.5,0.5,0.5,0.5,0.5,0.7,0.5,0.5,0.5,0.5,0.5]}}
     )
    )
    Index 0 Param 2=&Multiply_Float(
     A=&RandomRangeFloat(Minimum=0.7, Maximum=0.72)
     B=&SelectFloatFromArray(
      Direct Array Index=$Particles.SpawnGroup
      Float Selection Array=DataInterface() {propertyValues {FloatData [2,1.4,2,1.4,1.5,1.5,2,2,1.4,2,2,1.2,1.2,1]}}
     )
    )
    Index 0 Param 3=$System.FN_Integrator002.Value
    Index 1 Param 1=&SelectFloatFromArray(
     Direct Array Index=$Particles.SpawnGroup
     Float Selection Array=DataInterface() {propertyValues {FloatData [0,0,0,0,-3,-6,-5,-7,-10,-3,-5,-5,5,-3]}}
    )
    Index 1 Param 3=0.8
    Index 2 Param 1=&MakeFloatFromVector(Vector=$StackContext.InterpolateOverTime.InterpolateOverTime_V3_.Moving Average, Channel=::NewEnumerator0)
    Index 2 Param 2=&MakeFloatFromVector(Vector=$StackContext.InterpolateOverTime.InterpolateOverTime_V3_.Moving Average, Channel=::NewEnumerator1)
    Index 2 Param 3=&MakeFloatFromVector(Vector=$StackContext.InterpolateOverTime.InterpolateOverTime_V3_.Moving Average, Channel=::NewEnumerator2)
    Module.Index0Param0Write=true
    Module.Index0Param1Write=true
    Module.Index0Param2Write=true
    Module.Index0Param3Write=false
    Module.Index1Param0Write=true
    Module.Index1Param1Write=true
    Module.Index1Param2Write=true
    Module.Index1Param3Write=false
    Module.Index2Param0Write=true
    Module.Index2Param1Write=true
    Module.Index2Param2Write=true
    Module.Index2Param3Write=false
    Write Parameter Index 1=true
    Write Parameter Index 2=true
   ) {module=DynamicMaterialParameters001}
   &ScaleColor(Scale Alpha=&Lerp_Float(A=1, B=0, Alpha=$User.isOff), ScaleA=true, ScaleRGB=false) {module=ScaleColor002}
   &ScaleColor(
    Scale Alpha=&MakeCustomFloatFromBool(
     Bool=$User.bisFootHidden
     True Float=&SelectFloatFromArray(Direct Array Index=$Particles.SpawnGroup, Float Selection Array=DataInterface() {propertyValues {FloatData [1,1,1,1,0,0,0,0,1,0,0,1,1,0]}})
     False Float=1
    )
    ScaleA=true
    ScaleRGB=false
   ) {module=ScaleColor001}
  ]
  rendererClasses [&NiagaraMeshRendererProperties]
  renderers [{rendererClass=&NiagaraMeshRendererProperties, rendererIndex=0}]
 }
 UpperBody_Dots {
  emitterUpdateScript [
   &EmitterState(Inactive Response=::NewEnumerator1, Life Cycle Mode=::NewEnumerator1, Loop Delay=0.2, Loop Duration=1, UseLoopDelay=false)
   &SpawnRate(
    Spawn Probability=&Multiply_Float(A=0.9, B=&Lerp_Float(A=1, B=0, Alpha=$User.isOff))
    SpawnRate=&Multiply_Float(A=22, B=$System.DotsSpawnRateMult)
    Use Spawn Probability=true
   )
  ]
  particleSpawnScript [
   &FN_SetSpawnGroup(Number of Spawn Groups=19)
   &InitializeParticle(
    Color=<rgba 0 0 0 1>
    Lifetime=&Multiply_Float(A=&RandomRangeFloat(Minimum=1.5, Maximum=2), B=$System.DotsLifeTimeMult)
    Position Mode=::NewEnumerator0
    Sprite Size Mode=::NewEnumerator1
    Uniform Sprite Size Max=2
    Uniform Sprite Size Min=1.5
    Write Material Random=true
   ) {module=InitializeParticle001}
   &DynamicMaterialParameters(
    Index 0 Param 1=&UniformAOrBFloat(A=0.333, Distribution Weight=0.9, Evaluation Type=::NewEnumerator0, Recalculate Random Each Loop=true, Override Seed=false)
    Module.Index0Param0Write=true
    Module.Index0Param1Write=false
    Module.Index0Param2Write=false
    Module.Index0Param3Write=false
   )
   /SetVariables(Particles.RandomOffset:Vector3f=&Multiply_VectorByFloat(
     Vector=&RandomRangeVector(Minimum=<-1 -1 -1>, Maximum=<1 1 2>)
     Float=&RandomRangeFloat(Minimum=0, Maximum=2)
    )) {isSetParameters=true}
   &InitialMeshOrientation(
    Mesh Orientation Mode=::NewEnumerator7
    Rotation=&SelectVectorFromArray(
     Direct Array Index=$Particles.SpawnGroup
     Vector Selection Array=DataInterface() {propertyValues {FloatData [
        <0 0 0>
        <0 0 0>
        <0 0 0>
        <0 0 0>
        <0 0 0>
        <0 0 0>
        <0 0 0>
        <0 0 0>
        <0 0 0>
        <0 0 0>
        <0 0 0>
        <0 0 0>
        <0 0 0>
        <0 0 0>
        <0 0 0>
        <0 0 0>
        <0 0 0>
        <0 0 0>
        <0 0 0>
       ]}}
    )
    Use Rotation Vector=true
   )
  ]
  particleUpdateScript [
   &ParticleState() {module=ParticleState001}
   /SetVariables(Particles.InitialRotation:Quat4f=&LerpQuaternion(
     Quaternion A:Quat4f=<0 0 0 1>
     Quaternion B:Quat4f=$Particles.Initial.InitialRotation
     Lerp Factor:NiagaraFloat=&MakeCustomFloatFromBool(Bool=$User.IsSkydiving, True Float=1, False Float=0)
     Enable Spherical Linear Interpolation:NiagaraBool=false
    )) {isSetParameters=true}
   &SkeletalMeshLocation(
    Bone / Socket Index=&SelectIntFromArray(
     Direct Array Index=$Particles.SpawnGroup
     Int Selection Array=DataInterface() {propertyValues {IntData [9,10,11,12,13,14,15,16,17,18,19,20,21,24,25,26,27,28,29]}}
    )
    MaskSpawnGroup=false
    Mesh Sampling Type=::NewEnumerator5
    Sampled Position Offset=&RotateVectorByQuaternion(
     Vector To Rotate=&MakeVector(
      X=&Lerp_Float(
       A=0
       B=&Multiply_Float(
        A=&RandomRangeFloat(Minimum=2, Maximum=3)
        B=&SelectFloatFromArray(
         Direct Array Index=$Particles.SpawnGroup
         Float Selection Array=DataInterface() {propertyValues {FloatData [-0.5,-1,0,-0.5,-1,-0.2,-1,-1.1,-0.5,1.2,0,0,-1,-1,-1,-1,-1,-1,-1]}}
        )
       )
       Alpha=&FloatFromCurve(FloatCurve=DataInterface() {propertyValues {Curve {keys [
            {interpMode=RCIM_Cubic, tangentMode=RCTM_Break, time=0, value=0}
            {arriveTangent=1.4797, interpMode=RCIM_Cubic, tangentMode=RCTM_Break, time=1, value=1}
           ]}}})
      )
      Y=0
      Z=&Lerp_Float(
       A=0
       B=&Multiply_Float(
        A=&SelectFloatFromArray(
         Direct Array Index=$Particles.SpawnGroup
         Float Selection Array=DataInterface() {propertyValues {FloatData [0.7,1,0.5,0.5,1,0.5,0.7,1,0.5,0.6,0.6,0.6,0.8,0.6,0.6,0.8,0.6,0.8,0.6]}}
        )
        B=&RandomRangeFloat(Minimum=11, Maximum=13)
       )
       Alpha=&FloatFromCurve(FloatCurve=DataInterface() {propertyValues {Curve {keys [<at 0 0>,<at 1 1>]}}})
      )
     )
     Quaternion=$Particles.InitialRotation
    )
    Skeletal Mesh=$System.Skeletal Mesh
    Socket Sampling Mode=::NewEnumerator5
   )
   /SetVariables(Particles.Position=&AddVectorToPosition(Position=$Particles.Position, Vector=$Particles.RandomOffset)) {isSetParameters=true}
   &ParticleState()
   &CameraOffset(Camera Offset Amount=5)
   &ScaleColor(
    Scale Alpha=&Lerp_Float(
     A=1
     B=&SelectFloatFromArray(
      Direct Array Index=$Particles.SpawnGroup
      Float Selection Array=DataInterface() {propertyValues {FloatData [1,1,1,0,0,0,1,1,1,1,1,1,0,0,1,0,1,1,1]}}
     )
     Alpha=&MakeCustomFloatFromBool(Bool=$User.bisADS, True Float=1, False Float=0)
    )
    ScaleA=true
    ScaleRGB=false
   ) {module=ScaleColor001}
   /SetVariables(Particles.SpriteSize=&Multiply_Vector2D_ByFloat(
     Float=&Multiply_Float(
      A=1
      B=&UniformAOrBFloat(
       A=&FloatFromCurve(FloatCurve=DataInterface() {propertyValues {Curve {keys [
            <at 0 0 RCIM_Cubic>
            {arriveTangent=9.6335, interpMode=RCIM_Cubic, leaveTangent=2.7827, tangentMode=RCTM_Break, time=0.15, value=1}
            <at 0.4 0.6 RCIM_Cubic>
            {arriveTangent=2.7202, interpMode=RCIM_Cubic, leaveTangent=-2.3793, tangentMode=RCTM_Break, time=0.7, value=1}
            <at 1 0 RCIM_Cubic>
           ]}}})
       B=&FloatFromCurve(FloatCurve=DataInterface() {propertyValues {Curve {keys [
            <at 0 0 RCIM_Cubic>
            {arriveTangent=9.6335, interpMode=RCIM_Cubic, leaveTangent=2.7827, tangentMode=RCTM_Break, time=0.15, value=0.9794}
            {arriveTangent=0.0457, interpMode=RCIM_Cubic, leaveTangent=0.0457, time=0.5397, value=0.5}
            {arriveTangent=2.7202, interpMode=RCIM_Cubic, leaveTangent=-2.3793, tangentMode=RCTM_Break, time=0.6, value=1}
            <at 1 0 RCIM_Cubic>
           ]}}})
       Evaluation Type=::NewEnumerator0
       Recalculate Random Each Loop=true
       Override Seed=false
      )
     )
     Vector2D=$Particles.Initial.SpriteSize
    )) {isSetParameters=true}
   /SetVariables(
    Particles.AdditionalDot01Position:NiagaraPosition=&AddVectorToPosition(
     Position=$Particles.Position
     Vector=&Multiply_VectorByFloat(
      Vector=&RandomRangeVector(Minimum=<-1 -1 -1>, Maximum=<1 1 1>)
      Float=&Waveform(
       Global Input Position=&RandomRangeFloat(Minimum=0, Maximum=1)
       "Waveform [1]"=::NewEnumerator0
       "[1] Frequency"=1
       "[1] Amplitude Min/Max"=<0 1.5>
      )
     )
    )
    Particles.AdditionalDot01SpriteSize:Vector2f=&Multiply_Vector2D_ByFloat(
     Float=&Multiply_Float(
      A=&UniformAOrBFloat(
       A=&FloatFromCurve(FloatCurve=DataInterface() {propertyValues {Curve {keys [
            <at 0 0 RCIM_Cubic>
            {arriveTangent=9.6335, interpMode=RCIM_Cubic, leaveTangent=2.7827, tangentMode=RCTM_Break, time=0.15, value=1}
            <at 0.6943 0.7273 RCIM_Cubic>
            {arriveTangent=2.7202, interpMode=RCIM_Cubic, leaveTangent=-2.3793, tangentMode=RCTM_Break, time=0.8055, value=1}
            <at 0.95 0 RCIM_Cubic>
           ]}}})
       B=&FloatFromCurve(FloatCurve=DataInterface() {propertyValues {Curve {keys [
            <at 0 0 RCIM_Cubic>
            {arriveTangent=9.6335, interpMode=RCIM_Cubic, leaveTangent=2.7827, tangentMode=RCTM_Break, time=0.15, value=0.9794}
            {arriveTangent=0.0322, interpMode=RCIM_Cubic, leaveTangent=0.0322, time=0.5397, value=0.5}
            {arriveTangent=2.7202, interpMode=RCIM_Cubic, leaveTangent=-2.3793, tangentMode=RCTM_Break, time=0.7888, value=1}
            <at 0.95 0 RCIM_Cubic>
           ]}}})
       Evaluation Type=::NewEnumerator0
       Recalculate Random Each Loop=true
       Override Seed=false
      )
      B=1
     )
     Vector2D=&Vector2DFromFloat(Value=&RandomRangeFloat(Minimum=1, Maximum=1.5))
    )
    Particles.AdditionalDot01Visibility:NiagaraInt32=&MakeIntFromFloat(Float=&UniformAOrBFloat(A=0, Distribution Weight=0.6, Evaluation Type=::NewEnumerator0, Recalculate Random Each Loop=true, Override Seed=false))
   ) {isSetParameters=true}
   /SetVariables(
    Particles.AdditionalDot02Position:NiagaraPosition=&AddVectorToPosition(
     Position=$Particles.Position
     Vector=&Multiply_VectorByFloat(
      Vector=&RandomRangeVector(Minimum=<-1 -1 -1>, Maximum=<1 1 1>)
      Float=&Waveform(
       Global Input Position=&RandomRangeFloat(Minimum=0, Maximum=1)
       "Waveform [1]"=::NewEnumerator0
       "[1] Frequency"=1
       "[1] Amplitude Min/Max"=<0 2>
      )
     )
    )
    Particles.AdditionalDot02SpriteSize:Vector2f=&Multiply_Vector2D_ByFloat(
     Float=&Multiply_Float(
      A=&UniformAOrBFloat(
       A=&FloatFromCurve(FloatCurve=DataInterface() {propertyValues {Curve {keys [
            <at 0 0 RCIM_Cubic>
            {arriveTangent=9.6335, interpMode=RCIM_Cubic, leaveTangent=2.7827, tangentMode=RCTM_Break, time=0.15, value=1}
            <at 0.6943 0.7273 RCIM_Cubic>
            {arriveTangent=2.7202, interpMode=RCIM_Cubic, leaveTangent=-2.3793, tangentMode=RCTM_Break, time=0.8055, value=1}
            <at 0.95 0 RCIM_Cubic>
           ]}}})
       B=&FloatFromCurve(FloatCurve=DataInterface() {propertyValues {Curve {keys [
            <at 0 0 RCIM_Cubic>
            {arriveTangent=9.6335, interpMode=RCIM_Cubic, leaveTangent=2.7827, tangentMode=RCTM_Break, time=0.15, value=0.9794}
            {arriveTangent=0.0322, interpMode=RCIM_Cubic, leaveTangent=0.0322, time=0.5397, value=0.5}
            {arriveTangent=2.7202, interpMode=RCIM_Cubic, leaveTangent=-2.3793, tangentMode=RCTM_Break, time=0.7888, value=1}
            <at 0.95 0 RCIM_Cubic>
           ]}}})
       Evaluation Type=::NewEnumerator0
       Recalculate Random Each Loop=true
       Override Seed=false
      )
      B=1
     )
     Vector2D=&Vector2DFromFloat(Value=&RandomRangeFloat(Minimum=1.5, Maximum=2))
    )
    Particles.AdditionalDot02Visibility:NiagaraInt32=&MakeIntFromFloat(Float=&UniformAOrBFloat(A=0, Distribution Weight=0.88, Evaluation Type=::NewEnumerator0, Recalculate Random Each Loop=true, Override Seed=false))
   ) {isSetParameters=true}
   &ScaleColor(
    Scale Alpha=&MakeCustomFloatFromBool(
     Bool=$User.bisHandHidden
     True Float=&SelectFloatFromArray(
      Direct Array Index=$Particles.SpawnGroup
      Float Selection Array=DataInterface() {propertyValues {FloatData [1,1,1,1,1,1,1,0,1,1,0,0,0,0,1,1,1,1,1]}}
     )
     False Float=1
    )
    ScaleA=true
    ScaleRGB=false
   )
  ]
  rendererClasses [&NiagaraSpriteRendererProperties]
  renderers [
   {rendererClass=&NiagaraSpriteRendererProperties}
   {rendererClass=&NiagaraSpriteRendererProperties, rendererIndex=1}
   {rendererClass=&NiagaraSpriteRendererProperties, rendererIndex=2}
  ]
 }
 UpperBody_Flame {
  emitterUpdateScript [
   &EmitterState(
    Inactive Response=::NewEnumerator1
    Life Cycle Mode=::NewEnumerator1
    Loop Behavior=::NewEnumerator1
    Loop Duration Mode=::NewEnumerator1
    Scalability Mode=::NewEnumerator1
    UseLoopDelay=false
   )
   &SpawnBurst_Instantaneous(Spawn Count=19, Use Spawn Probability=false, UseLoopCountLimit=false)
  ]
  particleSpawnScript [
   &FN_SetSpawnGroup(Number of Spawn Groups=19)
   &InitializeParticle(
    Lifetime=5
    Mesh Scale Mode=::NewEnumerator3
    Mesh Uniform Scale=&Multiply_Float(
     A=0.5
     B=&SelectFloatFromArray(
      Direct Array Index=$Particles.SpawnGroup
      Direct Array Mode=::NewEnumerator1
      Float Selection Array=DataInterface() {propertyValues {FloatData [0.23,0.25,0.13,0.2,0.15,0.13,0.08,0.25,0.1332,0.13,0.11,0.2,0.25,0.13,0.0866,0.23,0.25,0.12,0.15]}}
     )
    )
    UsePositionOffset=false
    Write Material Random=true
   )
   &InitialMeshOrientation(
    Mesh Orientation Mode=::NewEnumerator7
    Rotation=&SelectVectorFromArray(
     Direct Array Index=$Particles.SpawnGroup
     Vector Selection Array=DataInterface() {propertyValues {FloatData [
        <0 0 0>
        <0 0 0>
        <0 0 0>
        <0 0 0>
        <0 0 0>
        <0 0 0>
        <0 0 0>
        <0 0 0>
        <0 0 0>
        <0 0 0>
        <0 0 0>
        <0 0 0>
        <0 0 0>
        <0 0 0>
        <0 0 0>
        <0 0 0>
        <0 0 0>
        <0 0 0>
        <0 0 0>
       ]}}
    )
    Rotation Coordinate Space=::NewEnumerator2
    Use Rotation Vector=true
   )
  ]
  particleUpdateScript [
   &ParticleState(Kill Particles When Lifetime Has Elapsed=false)
   /SetVariables(Particles.InitialRotation:Quat4f=&LerpQuaternion(
     Quaternion A:Quat4f=<0 0 0 1>
     Quaternion B:Quat4f=&LerpQuaternion(
      Quaternion A:Quat4f=<0 0 0 1>
      Quaternion B:Quat4f=$Particles.Initial.InitialRotation
      Lerp Factor:NiagaraFloat=&ClampFloat(Float=&Abs_Float(Float:NiagaraFloat=&Divide_Float(A=&MakeFloatFromVector(Vector=$Engine.Owner.Velocity, Channel=::NewEnumerator2), B=5000)))
      Enable Spherical Linear Interpolation:NiagaraBool=false
     )
     Lerp Factor:NiagaraFloat=&MakeCustomFloatFromBool(Bool=$User.IsSkydiving, True Float=1, False Float=0)
     Enable Spherical Linear Interpolation:NiagaraBool=false
    )) {isSetParameters=true}
   &SkeletalMeshLocation(
    Bone / Socket Index=&SelectIntFromArray(
     Direct Array Index=$Particles.SpawnGroup
     Int Selection Array=DataInterface() {propertyValues {IntData [9,10,11,12,13,14,15,16,17,18,19,20,21,24,25,26,27,28,29]}}
    )
    MaskSpawnGroup=false
    Mesh Sampling Type=::NewEnumerator5
    Sampled Position Offset Coordinate Space:ENiagaraSkelMeshTransforms=::NewEnumerator4
    Skeletal Mesh=$System.Skeletal Mesh
    Socket Sampling Mode=::NewEnumerator5
   )
   &InterpolateOverTime(Rate Of Change=10, Target Vector Value:Vector3f=$Particles.SkeletalMeshLocation.SampledPosition, Variable Type=::NewEnumerator2)
   &DynamicMaterialParameters(
    Index 0 Param 1=&Multiply_Float(
     A=0.7
     B=&SelectFloatFromArray(
      Direct Array Index=$Particles.SpawnGroup
      Float Selection Array=DataInterface() {propertyValues {FloatData [0.7,0.7,0.5,1,0.7,0.5,0.7,0.5,0.3,0,0.3,0.2,0.7,0.7,0.7,0.7,0.7,0.7,0.7]}}
     )
    )
    Index 0 Param 2=&Multiply_Float(
     A=0.7
     B=&SelectFloatFromArray(
      Direct Array Index=$Particles.SpawnGroup
      Float Selection Array=DataInterface() {propertyValues {FloatData [1.8,1.6,1.4,1.8,2,1.44,1.2,1.5,2,2,1,1.2,1.5,1.8,1.6,2,2,1.5,2]}}
     )
    )
    Index 0 Param 3=$System.FN_Integrator002.Value
    Index 1 Param 1=&SelectFloatFromArray(
     Direct Array Index=$Particles.SpawnGroup
     Float Selection Array=DataInterface() {propertyValues {FloatData [-70,-70,20,-60,-30,20,-10,-60,-30,30,-80,-80,30,50,-30,-50,-50,-50,-50]}}
    )
    Index 1 Param 3=&SelectFloatFromArray(
     Direct Array Index=$Particles.SpawnGroup
     Direct Array Mode=::NewEnumerator1
     Float Selection Array=DataInterface() {propertyValues {FloatData [0.3,1,0.4,0.3,1,0.4,0.6,0.4,1,1,0.3,0.3,0.5,1]}}
    )
    Index 2 Param 1=&MakeFloatFromVector(Vector=$StackContext.InterpolateOverTime.InterpolateOverTime_V3_.Moving Average, Channel=::NewEnumerator0)
    Index 2 Param 2=&MakeFloatFromVector(Vector=$StackContext.InterpolateOverTime.InterpolateOverTime_V3_.Moving Average, Channel=::NewEnumerator1)
    Index 2 Param 3=&MakeFloatFromVector(Vector=$StackContext.InterpolateOverTime.InterpolateOverTime_V3_.Moving Average, Channel=::NewEnumerator2)
    Module.Index0Param0Write=true
    Module.Index0Param1Write=true
    Module.Index0Param2Write=true
    Module.Index0Param3Write=false
    Module.Index1Param0Write=true
    Module.Index1Param1Write=true
    Module.Index1Param2Write=true
    Module.Index1Param3Write=false
    Module.Index2Param0Write=true
    Module.Index2Param1Write=true
    Module.Index2Param2Write=true
    Module.Index2Param3Write=false
    Write Parameter Index 1=true
    Write Parameter Index 2=true
   )
   &ScaleColor(
    Scale Alpha=&Multiply_Float(
     A=&Lerp_Float(
      A=1
      B=&SelectFloatFromArray(
       Direct Array Index=$Particles.SpawnGroup
       Float Selection Array=DataInterface() {propertyValues {FloatData [1,1,1,0,0,0,1,1,1,1,1,1,0,0,1,0,1,1,1]}}
      )
      Alpha=&MakeCustomFloatFromBool(Bool=$User.bisADS, True Float=1, False Float=0)
     )
     B=&RandomRangeFloat(Minimum=0.5, Maximum=0.8)
    )
    ScaleA=true
    ScaleRGB=false
   )
   &ScaleColor(
    Scale Alpha=&MakeCustomFloatFromBool(
     Bool=$User.bisHandHidden
     True Float=&SelectFloatFromArray(
      Direct Array Index=$Particles.SpawnGroup
      Float Selection Array=DataInterface() {propertyValues {FloatData [1,1,1,1,1,1,1,0,1,1,0,0,0,0,1,1,1,1,1]}}
     )
     False Float=1
    )
    ScaleA=true
    ScaleRGB=false
   ) {module=ScaleColor001}
   &ScaleColor(Scale Alpha=&Lerp_Float(A=1, B=0, Alpha=$User.isOff), ScaleA=true, ScaleRGB=false) {module=ScaleColor002}
  ]
  rendererClasses [&NiagaraMeshRendererProperties]
  renderers [{rendererClass=&NiagaraMeshRendererProperties, rendererIndex=0}]
 }
}
summary {
 systemName=NS_Character_SnapFreeze_Hunt_SL
 emitters [
  {emitterName=UpperBody_Dots, rendererClasses [&NiagaraSpriteRendererProperties]}
  {emitterName=LowerBody_Flame, rendererClasses [&NiagaraMeshRendererProperties]}
  {emitterName=LowerBody_Dots, rendererClasses [&NiagaraSpriteRendererProperties]}
  {bEnabled=false, emitterName=EyeGlow, rendererClasses [&NiagaraSpriteRendererProperties]}
  {bEnabled=false, emitterName=Eye_Flame, rendererClasses [&NiagaraMeshRendererProperties]}
  {bEnabled=false, emitterName=Dots_Eye, rendererClasses [&NiagaraSpriteRendererProperties]}
  {emitterName=UpperBody_Flame, rendererClasses [&NiagaraMeshRendererProperties]}
  {bEnabled=false, emitterName=FN_Debug, rendererClasses [&NiagaraSpriteRendererProperties]}
 ]
 userVariables [
  User.bisADS {type=NiagaraBool}
  User.bIsADSOther {type=NiagaraBool}
  User.bisFootHidden {type=NiagaraBool}
  User.bIsFrontEnd {type=NiagaraBool}
  User.bIsFrontEndPreview {type=NiagaraBool}
  User.bisHandHidden {type=NiagaraBool}
  User.ColorVariant {type=NiagaraFloat, defaultValue=0, description=""}
  User.EyeFlame_On/Off {type=NiagaraBool, defaultValue=true}
  User.EyeFlameAmount {type=NiagaraInt32, defaultValue=1}
  User.EyeFlameSide_Left/Right {type=NiagaraBool}
  User.isOff {type=NiagaraFloat, defaultValue=0, description=""}
  User.IsSkydiving {type=NiagaraBool}
 ]
}
