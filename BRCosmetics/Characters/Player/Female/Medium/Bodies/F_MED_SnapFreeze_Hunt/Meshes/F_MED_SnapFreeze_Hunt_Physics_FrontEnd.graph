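# Physics asset for the front-end version of the SnapFreeze Hunt female medium player body, covering the skirt and legs.
# Front and back skirts are three-capsule chains hung from tiny base bodies by angle-limited constraints, with the front allowed more swing than the back.
# Thigh, calf and foot capsules give the legs collision shapes, linked by constraints with free swing and twist.
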PhysicsAsset /BRCosmetics/Characters/Player/Female/Medium/Bodies/F_MED_SnapFreeze_Hunt/Meshes/F_MED_SnapFreeze_Hunt_Physics_FrontEnd
bodies {
 base_back_skirt [{shapeType=Capsule, length=0.505, radius=0.505, shapeName=base_back_skirt_capsule, rotation=<rot 0 -90 0>}]
 base_front_skirt [{shapeType=Capsule, length=0.505, radius=0.505, shapeName=base_front_skirt_capsule, rotation=<rot 0 -90 0>}]
 calf_l [{shapeType=Capsule, length=35, radius=6, shapeName=calf_l_capsule, center=<-23.0325 -3.7981 -1>, rotation=<rot 90 0 0>}]
 calf_r [{shapeType=Capsule, length=35, radius=6, shapeName=calf_l_capsule, center=<23.0325 3.7981 1>, rotation=<rot 90 0 0>}]
 dyn_back_skirt01 [{shapeType=Capsule, length=14, radius=6, shapeName=dyn_back_skirt01_capsule, center=<6.6062 -2.8 0>}]
 dyn_back_skirt02 [{shapeType=Capsule, length=22, radius=8, shapeName=dyn_back_skirt02_capsule, center=<6.6526 -4.5 0>}]
 dyn_back_skirt03 [{shapeType=Capsule, length=14, radius=8, shapeName=dyn_back_skirt03_capsule, center=<4.6415 -6.3 0>}]
 dyn_front_skirt01 [{shapeType=Capsule, length=22, radius=5.5, shapeName=dyn_front_skirt01_capsule, center=<5.0461 5 0>}]
 dyn_front_skirt02 [{shapeType=Capsule, length=9, radius=5.5, shapeName=dyn_front_skirt02_capsule, center=<5.2 5 0>}]
 dyn_front_skirt03 [{shapeType=Capsule, length=3, radius=3, shapeName=dyn_front_skirt03_capsule, center=<4.1317 2.6 0>}]
 foot_l [{shapeType=Capsule, length=11, radius=9, shapeName=foot_l_capsule, center=<-2.5828 6.1491 0>, rotation=<rot 0 0 90>}]
 foot_r [{shapeType=Capsule, length=11, radius=9, shapeName=foot_l_capsule, center=<2.5828 -6.1491 0>, rotation=<rot 0 0 90>}]
 thigh_l [{shapeType=Capsule, length=33, radius=8.5, shapeName=thigh_l_capsule, center=<-20 -1 1>, rotation=<rot 90 0 -5>}]
 thigh_r [{shapeType=Capsule, length=33, radius=8, shapeName=thigh_l_capsule, center=<20 1 -1>, rotation=<rot 90 180 175>}]
}
constraints [
 {
  bone1Name=foot_r
  bone2Name=calf_r
  swing1LimitDegrees=45
  swing1Motion=Free
  swing2LimitDegrees=45
  swing2Motion=Free
  twistLimitDegrees=45
  twistMotion=Free
 }
 {
  bone1Name=calf_r
  bone2Name=thigh_r
  swing1LimitDegrees=45
  swing1Motion=Free
  swing2LimitDegrees=45
  swing2Motion=Free
  twistLimitDegrees=45
  twistMotion=Free
 }
 {bone1Name=dyn_back_skirt03, bone2Name=dyn_back_skirt02, swing1LimitDegrees=20, swing2LimitDegrees=10, twistLimitDegrees=1}
 {bone1Name=dyn_front_skirt03, bone2Name=dyn_front_skirt02, swing1LimitDegrees=50, swing2LimitDegrees=40, twistLimitDegrees=15}
 {bone1Name=dyn_back_skirt02, bone2Name=dyn_back_skirt01, swing1LimitDegrees=25, swing2LimitDegrees=20, twistLimitDegrees=1}
 {bone1Name=dyn_front_skirt02, bone2Name=dyn_front_skirt01, swing1LimitDegrees=40, swing2LimitDegrees=40, twistLimitDegrees=1}
 {
  bone1Name=foot_l
  bone2Name=calf_l
  swing1LimitDegrees=45
  swing1Motion=Free
  swing2LimitDegrees=45
  swing2Motion=Free
  twistLimitDegrees=45
  twistMotion=Free
 }
 {
  bone1Name=calf_l
  bone2Name=thigh_l
  swing1LimitDegrees=45
  swing1Motion=Free
  swing2LimitDegrees=45
  swing2Motion=Free
  twistLimitDegrees=45
  twistMotion=Free
 }
 {bone1Name=dyn_front_skirt01, bone2Name=base_front_skirt, swing1LimitDegrees=50, swing2LimitDegrees=40, twistLimitDegrees=15}
 {bone1Name=dyn_back_skirt01, bone2Name=base_back_skirt, swing1LimitDegrees=20, swing2LimitDegrees=10, twistLimitDegrees=1}
]
