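# Animation Blueprint for the female SnapFreeze Hunt cosmetic body, adding secondary motion to the pose copied from its parent mesh.
# A Control Rig driven by skydive and stationary-crouch states feeds two AnimDynamics spring bones, then a cached pose.
# Two pelvis-space RigidBody sims, front-end and in-game, blend by bIsInFrontEnd before an additive back-skirt bone offset.
# The EventGraph updates their gravity override through a deprecated state function and tracks an unused parachute-surfing flag.
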AnimBlueprint /BRCosmetics/Characters/Player/Female/Medium/Bodies/F_MED_SnapFreeze_Hunt/Meshes/F_MED_SnapFreeze_Hunt_AnimBP
target_skeleton=&/F_MED_SnapFreeze_Hunt_Skeleton
graphs [
 AnimGraph {nodes [
   n1 = RigidBody!(
    ComponentPose=Local To Component(LocalPose=Use cached pose 'CACHE_Start'())
    Alpha
    OverrideWorldGravity:Vector=Get GravityOverride()
    ExternalForce
   ) -> (Pose) {
    bClampLinearTranslationLimitToRefPose=true
    bOverrideWorldGravity=true
    overridePhysicsAsset=&/F_MED_SnapFreeze_Hunt_Physics_FrontEnd
    simulationSpace=BaseBoneSpace
    alphaBoolBlend {blendOption=Linear}
    baseBoneRef {boneName=pelvis}
    componentAppliedLinearAccClamp=<3000 3000 3000>
    componentLinearAccScale=<0.6 0.6 0.6>
    componentLinearVelScale=<0.6 0.6 0.6>
    simSpaceSettings {
     maxAngularAcceleration=1000
     maxAngularVelocity=3
     maxLinearAcceleration=700
     maxLinearVelocity=500
     velocityScaleZ=0
     worldAlpha=0.2
     externalLinearDragV=<0.2 0.2>
    }
   }
   n2 = Component To Local(ComponentPose=@n1)
   n3 = RigidBody(ComponentPose=Local To Component(LocalPose=Use cached pose 'CACHE_Start'()), OverrideWorldGravity=Get GravityOverride()) {
    bClampLinearTranslationLimitToRefPose=true
    bOverrideWorldGravity=true
    overridePhysicsAsset=&/F_MED_SnapFreeze_Hunt_Physics
    simulationSpace=BaseBoneSpace
    alphaBoolBlend {blendOption=Linear}
    baseBoneRef {boneName=pelvis}
    componentAppliedLinearAccClamp=<3000 3000 3000>
    componentLinearAccScale=<0.6 0.6 0.6>
    componentLinearVelScale=<0.6 0.6 0.6>
    simSpaceSettings {
     maxAngularAcceleration=1000
     maxAngularVelocity=3
     maxLinearAcceleration=700
     maxLinearVelocity=500
     velocityScaleZ=0
     worldAlpha=0.2
     externalLinearDragV=<0.2 0.2>
    }
   }
   n4 = Component To Local(ComponentPose=@n3)
   n5 = Get bIsInFrontEnd()
   n6 = "Transform (Modify) Bone - Bone: base_back_skirt"!(
    ComponentPose=Local To Component(LocalPose=Blend Poses by bool(BlendPose_0=@n2, BlendPose_1=@n4, bActiveValue=@n5) {blendType=Linear})
    bAlphaBoolEnabled:bool=True
    Translation="0,0,-10"
    Rotation="0,-25,0"
    Scale="1,1,1"
   ) -> (Pose) {
    alphaInputType=Bool
    rotationMode=BMM_Additive
    rotationSpace=BCS_BoneSpace
    translationMode=BMM_Additive
    translationSpace=BCS_BoneSpace
    alphaBoolBlend {blendOption=QuadraticInOut, blendOutTime=5}
    boneToModify {boneName=base_back_skirt}
    rotation {yaw=-25}
    translation {z=-10}
   }
   Output Pose(Result=Component To Local(ComponentPose=@n6)) {layerGroup=None}
   n7 = Control Rig!(
    Source=Copy Pose From Mesh() {bUseAttachedParent=true}
    Alpha
    SkyDive:bool=Get bIsSkyDiving()
    StaticCrouching:bool=Get StaticCrouching() -> (StaticCrouching:bool)
   ) -> (Pose) {
    alphaBoolBlend {blendOption=Linear}
    controlRigAssetReference {blueprintRigClass=&/F_MED_SnapFreeze_Hunt_CtrlRig.F_MED_SnapFreeze_Hunt_CtrlRig_C}
   }
   n8 = Anim Dynamics - Bone: dyn_tennisspring_l(ComponentPose=Local To Component(LocalPose=@n7), Alpha=1, ExternalForce="0,0,0") {
    angularDampingOverride=1
    angularSpringConstant=20000
    bAngularSpring=true
    bOverrideAngularDamping=true
    bUseGravityOverride=true
    gravityScale=0.35
    simulationSpace=RootRelative
    alphaBoolBlend {blendOption=Linear}
    boundBone {boneName=dyn_tennisspring_l}
    componentLinearAccScale=<1 1 1>
    componentLinearVelScale=<1 1 1>
    gravityOverride {z=-980}
    physicsBodyDefinitions [{
      boundBone {boneName=dyn_tennisspring_l}
      boxExtents=<20 30 30>
      constraintSetup {angularLimitsMax=<0 0 20>, angularLimitsMin=<0 0 -20>, angularTarget=<1 0 0>}
      localJointOffset=<4 0 0>
     }]
   }
   n9 = Anim Dynamics - Bone: dyn_tennisspring_r(ComponentPose=@n8, Alpha=1, ExternalForce="0,0,0") {
    angularDampingOverride=1
    angularSpringConstant=20000
    bAngularSpring=true
    bOverrideAngularDamping=true
    bUseGravityOverride=true
    gravityScale=0.35
    simulationSpace=RootRelative
    alphaBoolBlend {blendOption=Linear}
    boundBone {boneName=dyn_tennisspring_r}
    componentLinearAccScale=<1 1 1>
    componentLinearVelScale=<1 1 1>
    gravityOverride {z=-980}
    physicsBodyDefinitions [{
      boundBone {boneName=dyn_tennisspring_r}
      boxExtents=<20 30 30>
      constraintSetup {angularLimitsMax=<0 0 20>, angularLimitsMin=<0 0 -20>, angularTarget=<1 0 0>}
      localJointOffset=<4 0 0>
     }]
   }
   Save cached pose 'CACHE_Start'(Pose=Component To Local(ComponentPose=@n9)) {cacheName=CACHE_Start}
  ]}
 EventGraph {nodes [
   Event BlueprintUpdateAnimation()
   > n1 = Sequence!(execute) -> (then_0, then_1, then_2:Exec)
   n2 = Get bIsMoving2D()
   Set StaticCrouching(execute=@n1.then_1, StaticCrouching:bool=AND Boolean(A=Get bIsCrouching(), B=NOT Boolean(A=@n2))) -> (Output_Get:bool)
   n3 = Gravity Override State Function DEPRECATED(
    execute=@n1.then_0
    CharacterPartAnimInstance=Self-Reference()
    GameTimeInSeconds=GetGameTimeInSeconds()
    GravityOverrideParams=Get Gravity Override Params()
   )
   > Set GravityOverride(GravityOverride=@n3)
   Event OnParachuteOpened()
   > n4 = Switch on EGliderType(Selection=Get PlayerGliderType() -> (PlayerGliderType=HangGlider))
   Set ParachutSurfing(execute=@n4.Surfing, ParachutSurfing:bool=true) -> (Output_Get:bool)
   Event BlueprintInitializeAnimation()
   > Set Skel Mesh(Skel Mesh=GetOwnerSkeletalMeshComponent())
   Event OnParachuteClosed()
   > Set ParachutSurfing(ParachutSurfing=false)
  ]}
]
variables [
 AnimationEvent {type=integer}
 Gravity Override Params {
  type=GravityOverrideParamsStruct
  default {skydive_FloatingDown {gravityOverride=<0 -700 -500>}, skydive_GoingUp {gravityOverride=<0 -300 -500>}}
 }
 GravityOverride {type=Vector}
 Leg_L Back {type=number}
 Leg_L Front {type=number}
 Leg_L Rot In {type=number}
 Leg_L Rot Out {type=number}
 Leg_R Back {type=number}
 Leg_R Front {type=number}
 Leg_R Rot In {type=number}
 Leg_R Rot Out {type=number}
 ParachutSurfing {type=boolean}
 RunAnimation {type=boolean}
 Skel Mesh {type=SkeletalMeshComponent}
 StaticCrouching {type=boolean}
]
