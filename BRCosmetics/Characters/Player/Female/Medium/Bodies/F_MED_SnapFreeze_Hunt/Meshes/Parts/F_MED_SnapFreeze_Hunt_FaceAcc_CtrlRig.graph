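# Control Rig for the SnapFreeze Hunt female FaceAcc part that repositions the ponytail base bone as the head turns and tilts.
# DeltaRotation swing-twist decomposes head rotation relative to head_null, then remaps one Euler axis to a clamped 0–1 weight.
# Three weighted ModifyTransforms offset base_ponytail for twist in either direction and swing in one.
# A Controllers function can drive the head from head_ctrl for testing, but is switched off.
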ControlRigBlueprint /BRCosmetics/Characters/Player/Female/Medium/Bodies/F_MED_SnapFreeze_Hunt/Meshes/Parts/F_MED_SnapFreeze_Hunt_FaceAcc_CtrlRig
graphs [
 New Function_ContainedGraph {nodes [
   n1 = RigVMFunctionEntryNode()
   RigVMFunction_ControlFlowBranch = &RigVMFunction_ControlFlowBranch(ExecuteContext=@n1.ExecuteContext, Condition=@n1.exec)
   RigVMFunctionReturnNode(ExecuteContext=@RigVMFunction_ControlFlowBranch.Completed)
   &RigUnit_SetTransform(
    Item=(Type=Bone,Name="head")
    Weight=1
    bInitial=False
    Value=&RigUnit_GetTransform(Item=(Type=Control,Name="head_ctrl"), bInitial=False).Transform
    ExecutePin=@RigVMFunction_ControlFlowBranch.True
   )
  ]}
 New Function_ContainedGraph {nodes [
   n1 = RigVMFunctionEntryNode()
   RigVMFunction_MathBoolToInteger = &RigVMFunction_MathBoolToInteger(Value=@n1.onlyTwist)
   RigVMFunction_MathQuaternionSwingTwist = &RigVMFunction_MathQuaternionSwingTwist(
    TwistAxis=@n1.twistAxis
    Input=&RigUnit_GetRelativeTransformForItem(
     Child=(Type=Bone,Name="None")
     Parent=(Type=Bone,Name="None")
     bChildInitial=False
     bParentInitial=False
     Child.Name=@n1.child
     Parent.Name=@n1.parent
     Parent.Type=@n1.parentType
    ).RelativeTransform.Rotation
   )
   RigVMFunction_MathQuaternionToEuler = &RigVMFunction_MathQuaternionToEuler(
    Value=RigVMDispatchNode(
     Values=((),())
     Index=@RigVMFunction_MathBoolToInteger.Result
     Values.0=@RigVMFunction_MathQuaternionSwingTwist.Swing
     Values.1=@RigVMFunction_MathQuaternionSwingTwist.Twist
    ) {template="DISPATCH_RigVMDispatch_SelectInt32(in Index,in Values,out Result)"}.Result
    RotationOrder=@n1.Rotation Order
   )
   > DISPATCH_RigVMDispatch_Print = RigVMDispatchNode(ScreenColor=(R=1,G=1,B=1,A=1), ScreenDuration=1.5, Enabled=@n1.debugRotation, Value=@RigVMFunction_MathQuaternionToEuler.Result) {template="DISPATCH_RigVMDispatch_Print(in Prefix,in Value,in Enabled,in ScreenDuration,in ScreenColor)"}
   Clamp = &RigVMFunction_MathIntClamp(Value=@n1.primAxis)
   DISPATCH_RigVMDispatch_SelectInt32 = RigVMDispatchNode(
    Values=(0,0,0)
    Index=@Clamp.Result
    Values.0=@RigVMFunction_MathQuaternionToEuler.Result.X
    Values.1=@RigVMFunction_MathQuaternionToEuler.Result.Y
    Values.2=@RigVMFunction_MathQuaternionToEuler.Result.Z
   ) {template="DISPATCH_RigVMDispatch_SelectInt32(in Index,in Values,out Result)"}
   Remap = &RigVMFunction_MathDoubleRemap(
    TargetMaximum=1
    TargetMinimum=0
    bClamp=False
    Value=@DISPATCH_RigVMDispatch_SelectInt32.Result
    SourceMaximum=@n1.max
    SourceMinimum=@n1.min
   )
   > RigVMFunctionReturnNode(outputWeight=&RigVMFunction_MathDoubleClamp(Value=@Remap.Result).Result)
  ]}
 RigVMFunctionLibrary {nodes [
   Controllers = RigVMCollapseNode(ExecuteContext=(), exec=False)
   DeltaRotation = RigVMCollapseNode(ExecuteContext=(), Rotation Order=ZYX, twistAxis=(X=1,Y=0,Z=0))
  ]}
 RigVMModel {nodes [
   RigUnit_BeginExecution = &RigUnit_BeginExecution()
   Controllers = RigVMFunctionReferenceNode(exec=false, ExecuteContext=@RigUnit_BeginExecution.ExecutePin)
   > DeltaRotation = RigVMFunctionReferenceNode(Rotation Order=ZYX, child=head, max=32, onlyTwist=true, parent=head_null, parentType=Null, twistAxis=(X=1,Y=0,Z=0))
   > &RigUnit_ModifyTransforms(
    ItemToModify=((Item=(Type=Bone,Name="base_ponytail"),Transform=(Rotation=(X=0.2096,Y=0.105,Z=0.0226,W=-0.9719),Translation=(X=0.6308,Y=0.546,Z=-2.8865),Scale3D=(X=1,Y=1,Z=1))))
    WeightMinimum=0
    Weight=@DeltaRotation.outputWeight
   )
   > DeltaRotation_2 = RigVMFunctionReferenceNode(Rotation Order=ZYX, child=head, max=-32, onlyTwist=true, parent=head_null, parentType=Null, twistAxis=(X=1,Y=0,Z=0))
   > &RigUnit_ModifyTransforms(
    ItemToModify=((Item=(Type=Bone,Name="base_ponytail"),Transform=(Rotation=(X=-0.0591,Y=0.0002,Z=-0.0028,W=0.9982),Translation=(X=-0.0025,Y=-2.9064,Z=0.5065),Scale3D=(X=1,Y=1,Z=1))))
    WeightMinimum=0
    Weight=@DeltaRotation_2.outputWeight
   )
   > DeltaRotation_1 = RigVMFunctionReferenceNode(Rotation Order=ZYX, child=head, max=17, onlyTwist=false, parent=head_null, parentType=Null, primAxis=2, twistAxis=(X=1,Y=0,Z=0))
   > &RigUnit_ModifyTransforms(
    ItemToModify=((Item=(Type=Bone,Name="base_ponytail"),Transform=(Rotation=(X=-0.146,Y=-0.0594,Z=0.0895,W=0.9834),Translation=(X=0.45,Y=-2.3,Z=-0.13),Scale3D=(X=1,Y=1,Z=1))))
    WeightMinimum=0
    Weight=@DeltaRotation_1.outputWeight
   )
  ]}
]
hierarchy [
 root {loc=<0 0 0>, rot=<0 0 0 1>}
 pelvis {parent=root, loc=<0 -0.5301 98.0571>, rot=<0.0127 0.707 -0.0127 -0.707>}
 spine_01 {parent=pelvis, loc=<6.1208 -0.9328 0>, rot=<0 0 0.034 -0.9994>}
 spine_02 {parent=spine_01, loc=<6.5996 -0.2095 0>, rot=<0 0 0.0238 -0.9997>}
 spine_03 {parent=spine_02, loc=<6.4914 -0.1237 0>, rot=<0 0 0.0149 -0.9999>}
 spine_04 {parent=spine_03, loc=<6.9598 0.0794 0>, rot=<0 0 -0.0132 -0.9999>}
 spine_05 {parent=spine_04, loc=<6.9987 0 0>, rot=<0 0 -0.0015 -1>}
 clavicle_l {parent=spine_05, loc=<5.4831 -0.1053 -3.2946>, rot=<0.0465 -0.7289 -0.009 -0.6829>}
 upperarm_l {parent=clavicle_l, loc=<9.8658 -1.2136 -0.4105>, rot=<0.0164 -0.3245 -0.0269 -0.9454>}
 neck_01 {parent=spine_05, loc=<11.3138 0.1529 0>, rot=<0 0 -0.2003 -0.9797>}
 neck_02 {parent=neck_01, loc=<4.5627 0 0>, rot=<0 0 0.0504 -0.9987>}
 head {parent=neck_02, loc=<5.3914 0 0>, rot=<0 0 0.115 -0.9934>}
 base_hair {parent=head, loc=<5.6583 -1.5909 0>, rot=<0.0041 -1 0 0>}
 dyn_hair {parent=base_hair, loc=<0 0 0>, rot=<0 0 0 1>}
 base_earrin_l {parent=dyn_hair, loc=<5.3504 1.5461 6.2829>, rot=<0.3827 0 0 -0.9239>}
 dyn_earrin_l {parent=base_earrin_l, loc=<0 0 0>, rot=<0 0 0 1>}
 base_earrin_r {parent=dyn_hair, loc=<5.3504 1.5461 -6.2829>, rot=<0.3827 0 0 -0.9239>}
 dyn_earrin_r {parent=base_earrin_r, loc=<0 0 0>, rot=<0 0 0 1>}
 base_ponytail {parent=head, loc=<3.4762 -8.7082 -3.4205>, rot=<-0.2045 -0.84 -0.2948 0.4069>}
 dyn_ponytail01 {parent=base_ponytail, loc=<0 0 0>, rot=<0 0 0 1>}
 dyn_ponytail02 {parent=dyn_ponytail01, loc=<7.919 -0.0641 0.0299>, rot=<0.1096 0.1045 -0.1554 -0.9762>}
 dyn_ponytail03 {parent=dyn_ponytail02, loc=<6.0523 0.0128 0.0058>, rot=<-0.0019 0.2533 0.014 -0.9673>}
 dyn_ponytail04 {parent=dyn_ponytail03, loc=<6.5352 -0.0451 0.0188>, rot=<0.0178 0.0679 0.1387 -0.9878>}
 dyn_ponytail04b {parent=dyn_ponytail04, loc=<6.4996 -0.0209 0.008>, rot=<0.002 -0.0291 0.0872 -0.9958>}
 dyn_ponytail05 {parent=dyn_ponytail04b, loc=<5.6085 -0.0116 0.0194>, rot=<0.0222 -0.1087 0.1019 -0.9886>}
 dyn_ponytail06 {parent=dyn_ponytail05, loc=<5.8542 -0.0051 0.0482>, rot=<-0.0005 0.0024 0.1244 -0.9922>}
 dyn_ponytail06_b {parent=dyn_ponytail06, loc=<2.8843 -0.1759 -0.6593>, rot=<0.0493 -0.1499 0.2024 -0.9665>}
 dyn_ponytail06_a {parent=dyn_ponytail06, loc=<3.551 0.1318 0.5474>, rot=<0.0056 0.1799 0.1636 -0.97>}
 head_null {parent=neck_02, type=Null, loc=<5.3914 0 0>, rot=<0 0 0.115 -0.9934>}
 clavicle_r {parent=spine_05, loc=<5.4831 -0.1053 3.2946>, rot=<0.7289 0.0465 -0.6829 0.009>}
 upperarm_r {parent=clavicle_r, loc=<-9.8658 1.2136 0.4105>, rot=<0.0164 -0.3245 -0.0269 -0.9454>}
 head_ctrl {controlType=EULER_TRANSFORM, type=Control, loc=<0 0 0>, rot=<0 0 0 1>}
]
