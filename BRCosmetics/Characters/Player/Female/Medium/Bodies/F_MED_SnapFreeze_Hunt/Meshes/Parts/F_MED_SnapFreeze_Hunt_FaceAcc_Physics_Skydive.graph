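# Skydive physics asset for the F_MED_SnapFreeze_Hunt face accessory, defining capsule bodies and joints for a ponytail and two earrings.
# Both hang from base bones linked through base_hair to the head, with body capsules on neck, upper spine and only the right clavicle and upper arm.
# Anchor and body joints swing and twist freely.
# Ponytail limits are very tight near the root, loosening toward its forked tip, while earrings allow moderate swing.
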PhysicsAsset /BRCosmetics/Characters/Player/Female/Medium/Bodies/F_MED_SnapFreeze_Hunt/Meshes/Parts/F_MED_SnapFreeze_Hunt_FaceAcc_Physics_Skydive
bodies {
 base_earrin_l [{shapeType=Capsule, length=0.5, radius=0.5, shapeName=base_earrin_l_capsule}]
 base_earrin_r [{shapeType=Capsule, length=0.5, radius=0.5, shapeName=base_earrin_r_capsule}]
 base_hair [{shapeType=Capsule, length=0.5, radius=0.5, shapeName=base_hair_capsule}]
 base_ponytail [{shapeType=Capsule, length=0.5, radius=0.5, shapeName=base_ponytail_capsule}]
 clavicle_r [{shapeType=Capsule, length=33, radius=6.5, shapeName=clavicle_l_capsule, center=<-5 0.9028 0>, rotation=<rot 90 0 180>}]
 dyn_earrin_l [{shapeType=Capsule, length=4, radius=1, shapeName=dyn_earrin_l_capsule, center=<2.3269 0.0297 -0.0497>, rotation=<rot 90 0 -10>}]
 dyn_earrin_r [{shapeType=Capsule, length=4, radius=1, shapeName=dyn_earrin_r_capsule, center=<2.3275 0.0003 -0.0007>, rotation=<rot 90 0 -10>}]
 dyn_ponytail01 [{shapeType=Capsule, length=6, radius=1.1, shapeName=dyn_ponytail01_capsule, center=<3 -0.035 0>, rotation=<rot 90 0 0>}]
 dyn_ponytail02 [{shapeType=Capsule, length=6, radius=1.5, shapeName=dyn_ponytail02_capsule, center=<3 0.2 0>, rotation=<rot 90 0 0>}]
 dyn_ponytail03 [{shapeType=Capsule, length=6, radius=1.3, shapeName=dyn_ponytail03_capsule, center=<3 -0.2 0>, rotation=<rot 90 0 0>}]
 dyn_ponytail04 [{shapeType=Capsule, length=6, radius=1.3, shapeName=dyn_ponytail04_capsule, center=<3.0142 -0.2 0>, rotation=<rot 90 180 175>}]
 dyn_ponytail04b [{shapeType=Capsule, length=5, radius=1, shapeName=dyn_ponytail04b_capsule, center=<1.8958 0.2 0>, rotation=<rot 90 180 177>}]
 dyn_ponytail05 [{shapeType=Capsule, length=6, radius=1.1, shapeName=dyn_ponytail05_capsule, center=<3 0.4 0>, rotation=<rot 90 0 0>}]
 dyn_ponytail06 [{shapeType=Capsule, length=6, radius=1.1, shapeName=dyn_ponytail06_capsule, center=<3 0.1 0>, rotation=<rot 90 0 0>}]
 dyn_ponytail06_a [{shapeType=Capsule, length=5.5, radius=1.1, shapeName=dyn_ponytail06_a_capsule, center=<2.5 0.5 0>, rotation=<rot 90 0 0>}]
 dyn_ponytail06_b [{shapeType=Capsule, length=4.5, radius=1.1, shapeName=dyn_ponytail06_b_capsule, center=<2.2 0.5 0>, rotation=<rot 90 0 0>}]
 head [{shapeType=Capsule, length=11, radius=5.5, shapeName=head_capsule, rotation=<rot 90 0 0>}]
 neck_01 [{shapeType=Capsule, length=6, radius=4.5, shapeName=neck_01_capsule, rotation=<rot 90 0 0>}]
 spine_03 [{shapeType=Capsule, length=14, radius=6, shapeName=spine_03_capsule, center=<2.7147 4.689 0>}]
 spine_04 [{shapeType=Capsule, length=15, radius=6, shapeName=spine_04_capsule, center=<2.0569 6.8088 0>}]
 spine_05 [{shapeType=Capsule, length=26, radius=6, shapeName=spine_05_capsule, center=<0 3.2924 0>}]
 upperarm_r [{shapeType=Capsule, length=22, radius=6.5, shapeName=upperarm_l_capsule, center=<-11.7008 1.7147 0>, rotation=<rot 90 0 180>}]
}
constraints [
 {
  bone1Name=base_hair
  bone2Name=head
  swing1LimitDegrees=45
  swing1Motion=Free
  swing2LimitDegrees=45
  swing2Motion=Free
  twistLimitDegrees=45
  twistMotion=Free
 }
 {
  bone1Name=base_earrin_l
  bone2Name=base_hair
  swing1LimitDegrees=45
  swing1Motion=Free
  swing2LimitDegrees=45
  swing2Motion=Free
  twistLimitDegrees=45
  twistMotion=Free
 }
 {bone1Name=dyn_earrin_l, bone2Name=base_earrin_l, swing1LimitDegrees=15, swing2LimitDegrees=10, twistLimitDegrees=1}
 {
  bone1Name=base_earrin_r
  bone2Name=base_hair
  swing1LimitDegrees=45
  swing1Motion=Free
  swing2LimitDegrees=45
  swing2Motion=Free
  twistLimitDegrees=45
  twistMotion=Free
 }
 {bone1Name=dyn_earrin_r, bone2Name=base_earrin_r, swing1LimitDegrees=15, swing2LimitDegrees=10, twistLimitDegrees=1}
 {
  bone1Name=base_ponytail
  bone2Name=base_hair
  swing1LimitDegrees=45
  swing1Motion=Free
  swing2LimitDegrees=45
  swing2Motion=Free
  twistLimitDegrees=45
  twistMotion=Free
 }
 {bone1Name=dyn_ponytail01, bone2Name=base_ponytail, swing1LimitDegrees=1, swing2LimitDegrees=1, twistLimitDegrees=1}
 {bone1Name=dyn_ponytail02, bone2Name=dyn_ponytail01, swing1LimitDegrees=1, swing2LimitDegrees=1, twistLimitDegrees=1}
 {bone1Name=dyn_ponytail03, bone2Name=dyn_ponytail02, swing1LimitDegrees=1, swing2LimitDegrees=1, twistLimitDegrees=1}
 {bone1Name=dyn_ponytail04, bone2Name=dyn_ponytail03, swing1LimitDegrees=1, swing2LimitDegrees=1, twistLimitDegrees=1}
 {bone1Name=dyn_ponytail05, bone2Name=dyn_ponytail04b, swing1LimitDegrees=5, swing2LimitDegrees=5, twistLimitDegrees=1}
 {bone1Name=dyn_ponytail06, bone2Name=dyn_ponytail05, swing1LimitDegrees=5, swing2LimitDegrees=5, twistLimitDegrees=1}
 {bone1Name=dyn_ponytail06_a, bone2Name=dyn_ponytail06, swing1LimitDegrees=10, swing2LimitDegrees=10, twistLimitDegrees=1}
 {bone1Name=dyn_ponytail06_b, bone2Name=dyn_ponytail06, swing1LimitDegrees=10, swing2LimitDegrees=10, twistLimitDegrees=1}
 {
  bone1Name=head
  bone2Name=neck_01
  swing1LimitDegrees=45
  swing1Motion=Free
  swing2LimitDegrees=45
  swing2Motion=Free
  twistLimitDegrees=45
  twistMotion=Free
 }
 {
  bone1Name=spine_04
  bone2Name=spine_03
  swing1LimitDegrees=45
  swing1Motion=Free
  swing2LimitDegrees=45
  swing2Motion=Free
  twistLimitDegrees=45
  twistMotion=Free
 }
 {
  bone1Name=spine_05
  bone2Name=spine_04
  swing1LimitDegrees=45
  swing1Motion=Free
  swing2LimitDegrees=45
  swing2Motion=Free
  twistLimitDegrees=45
  twistMotion=Free
 }
 {
  bone1Name=clavicle_r
  bone2Name=spine_05
  swing1LimitDegrees=45
  swing1Motion=Free
  swing2LimitDegrees=45
  swing2Motion=Free
  twistLimitDegrees=45
  twistMotion=Free
 }
 {
  bone1Name=upperarm_r
  bone2Name=clavicle_r
  swing1LimitDegrees=45
  swing1Motion=Free
  swing2LimitDegrees=45
  swing2Motion=Free
  twistLimitDegrees=45
  twistMotion=Free
 }
 {
  bone1Name=neck_01
  bone2Name=spine_05
  swing1LimitDegrees=45
  swing1Motion=Free
  swing2LimitDegrees=45
  swing2Motion=Free
  twistLimitDegrees=45
  twistMotion=Free
 }
 {bone1Name=dyn_ponytail04b, bone2Name=dyn_ponytail04, swing1LimitDegrees=1, swing2LimitDegrees=1, twistLimitDegrees=1}
]
