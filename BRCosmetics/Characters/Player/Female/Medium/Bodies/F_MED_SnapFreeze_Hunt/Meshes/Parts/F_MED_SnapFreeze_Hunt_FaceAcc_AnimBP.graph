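# Animation Blueprint for the SnapFreeze Hunt female face accessory part, adding secondary physics motion to its earrings and other bones.
# It copies the parent pose, layers in body-mesh clavicles and upper arms, then runs spring AnimDynamics on both earring bones.
# That cached pose feeds three neck-relative RigidBody setups, blended for normal, skydiving and front-end (after Control Rig) states.
# The event graph updates gravity and linear scales, resetting dynamics when customization loads.
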AnimBlueprint /BRCosmetics/Characters/Player/Female/Medium/Bodies/F_MED_SnapFreeze_Hunt/Meshes/Parts/F_MED_SnapFreeze_Hunt_FaceAcc_AnimBP
target_skeleton=&/F_MED_SnapFreeze_Hunt_FaceAcc_Skeleton
graphs [
 AnimGraph {nodes [
   n1 = Get Linear Scale Free() -> (Linear Scale Free:Vector)
   n2 = RigidBody!(
    ComponentPose=Local To Component(LocalPose=Use cached pose 'Cache_AnimDynamics'())
    Alpha
    OverrideWorldGravity:Vector=Get GravityOverride()
    ExternalForce
    ComponentLinearAccScale:Vector=@n1
    ComponentLinearVelScale:Vector=@n1
   ) -> (Pose) {
    bClampLinearTranslationLimitToRefPose=true
    bOverrideWorldGravity=true
    overridePhysicsAsset=&/F_MED_SnapFreeze_Hunt_FaceAcc_Physics
    simulationSpace=BaseBoneSpace
    alphaBoolBlend {blendOption=Linear}
    baseBoneRef {boneName=neck_01}
    componentAppliedLinearAccClamp=<2000 2000 2000>
    componentLinearAccScale=<0.6 0.2 0.2>
    componentLinearVelScale=<0.6 0.2 0.2>
    simSpaceSettings {worldAlpha=0.05}
   }
   n3 = Component To Local(ComponentPose=@n2)
   n4 = Get Linear Scale Free()
   n5 = RigidBody(
    ComponentPose=Local To Component(LocalPose=Use cached pose 'Cache_AnimDynamics'())
    OverrideWorldGravity=Get GravityOverride()
    ComponentLinearAccScale=@n4
    ComponentLinearVelScale=@n4
   ) {
    bClampLinearTranslationLimitToRefPose=true
    bOverrideWorldGravity=true
    overridePhysicsAsset=&/F_MED_SnapFreeze_Hunt_FaceAcc_Physics_Skydive
    simulationSpace=BaseBoneSpace
    alphaBoolBlend {blendOption=Linear}
    baseBoneRef {boneName=neck_01}
    componentAppliedLinearAccClamp=<2000 2000 2000>
    componentLinearAccScale=<0.6 0.2 0.2>
    componentLinearVelScale=<0.6 0.2 0.2>
    simSpaceSettings {worldAlpha=0.05}
   }
   n6 = Component To Local(ComponentPose=@n5)
   n7 = Get bIsSkyDiving()
   n8 = Control Rig(Source=Use cached pose 'Cache_AnimDynamics'()) {
    alphaBoolBlend {blendOption=Linear}
    controlRigAssetReference {blueprintRigClass=&/F_MED_SnapFreeze_Hunt_FaceAcc_CtrlRig.F_MED_SnapFreeze_Hunt_FaceAcc_CtrlRig_C}
   }
   n9 = Get Linear Scale Free()
   n10 = RigidBody(
    ComponentPose=Local To Component(LocalPose=@n8)
    OverrideWorldGravity=Get GravityOverride()
    ComponentLinearAccScale=@n9
    ComponentLinearVelScale=@n9
   ) {
    bClampLinearTranslationLimitToRefPose=true
    bOverrideWorldGravity=true
    overridePhysicsAsset=&/F_MED_SnapFreeze_Hunt_FaceAccFE_Physics
    simulationSpace=BaseBoneSpace
    alphaBoolBlend {blendOption=Linear}
    baseBoneRef {boneName=neck_01}
    componentAppliedLinearAccClamp=<2000 2000 2000>
    componentLinearAccScale=<0.6 0.2 0.2>
    componentLinearVelScale=<0.6 0.2 0.2>
    simSpaceSettings {worldAlpha=0.05}
   }
   Output Pose(Result=Blend Poses by bool(
     BlendPose_0=Component To Local(ComponentPose=@n10)
     BlendPose_1=Blend Poses by bool(BlendPose_0=@n6, BlendPose_1=@n3, bActiveValue=@n7) {blendType=Linear}
     bActiveValue=Get bIsInFrontEnd()
    ) {blendType=Linear}) {layerGroup=None}
   n11 = Anim Dynamics - Bone: base_earrin_l(ComponentPose=Local To Component(LocalPose=Use cached pose 'Cache_StartPose'()), Alpha=1, ExternalForce="0,0,0") {
    angularDampingOverride=1
    angularSpringConstant=5000
    bAngularSpring=true
    bOverrideAngularDamping=true
    bUseGravityOverride=true
    gravityScale=0.35
    simulationSpace=RootRelative
    alphaBoolBlend {blendOption=Linear}
    boundBone {boneName=base_earrin_l}
    componentLinearAccScale=<1 1 1>
    componentLinearVelScale=<1 1 1>
    gravityOverride {z=-980}
    physicsBodyDefinitions [{
      boundBone {boneName=base_earrin_l}
      boxExtents=<5 10 10>
      constraintSetup {angularLimitsMax=<10 60 60>, angularLimitsMin=<-10 -60 -60>, angularTarget=<1 0 0>}
      localJointOffset=<1 0 0>
     }]
   }
   n12 = Anim Dynamics - Bone: base_earrin_r(ComponentPose=@n11, Alpha=1, ExternalForce="0,0,0") {
    angularDampingOverride=1
    angularSpringConstant=5000
    bAngularSpring=true
    bOverrideAngularDamping=true
    bUseGravityOverride=true
    gravityScale=0.35
    simulationSpace=RootRelative
    alphaBoolBlend {blendOption=Linear}
    boundBone {boneName=base_earrin_r}
    componentLinearAccScale=<1 1 1>
    componentLinearVelScale=<1 1 1>
    gravityOverride {z=-980}
    physicsBodyDefinitions [{
      boundBone {boneName=base_earrin_r}
      boxExtents=<5 10 10>
      constraintSetup {angularLimitsMax=<10 60 60>, angularLimitsMin=<-10 -60 -60>, angularTarget=<1 0 0>}
      localJointOffset=<1 0 0>
     }]
   }
   Save cached pose 'Cache_AnimDynamics'(Pose=Component To Local(ComponentPose=@n12)) {cacheName=Cache_AnimDynamics}
   n13 = Layered blend per bone(
    BasePose=Copy Pose From Mesh() {bCopyCurves=true, bUseAttachedParent=true}
    BlendPoses_0=Copy Pose From Mesh(SourceMeshComponent=Get Body_Mesh())
   ) {layerSetup [{branchFilters [
       {boneName=clavicle_l}
       {boneName=clavicle_r}
       {boneName=upperarm_l}
       {boneName=upperarm_r}
       {boneName=upperarm_twist_01_l}
       {boneName=upperarm_twist_01_r}
       {boneName=deform_upperarm_bk_base_l}
       {boneName=deform_upperarm_bk_l}
       {boneName=deform_upperarm_fr_base_l}
       {boneName=deform_upperarm_fr_l}
       {boneName=deform_upperarm_up_base_l}
       {boneName=deform_upperarm_up_l}
       {boneName=deform_upperarm_bk_base_r}
       {boneName=deform_upperarm_bk_r}
       {boneName=deform_upperarm_fr_base_r}
       {boneName=deform_upperarm_fr_r}
       {boneName=deform_upperarm_up_base_r}
       {boneName=deform_upperarm_up_r}
      ]}]}
   Save cached pose 'Cache_StartPose'(Pose=@n13) {cacheName=Cache_StartPose}
  ]}
 EventGraph {nodes [
   Event BlueprintUpdateAnimation()
   > n1 = Sequence!(execute) -> (then_0, then_1, then_2:Exec)
   n2 = Branch(execute=@n1.then_1, Condition=Get bIsSkyDiving())
   > Set Linear Scale Free(Linear Scale Free:Vector="0.35,0.25,0.25") -> (Output_Get:Vector)
   > Set Linear Scale Backpack(Linear Scale Backpack:Vector="0.3,0.2,0.2") -> (Output_Get:Vector)
   > Set Linear Scale Cape(Linear Scale Cape:Vector="0.25,0.2,0.2") -> (Output_Get:Vector)
   Set Linear Scale Free(execute=@n2.else, Linear Scale Free="0.32,0.35,0.32")
   > Set Linear Scale Backpack(Linear Scale Backpack="0.2,0.2,0.2")
   > Set Linear Scale Cape(Linear Scale Cape="0.25,0.2,0.2")
   n3 = Gravity Override State Function DEPRECATED(
    execute=@n1.then_0
    CharacterPartAnimInstance=Self-Reference()
    GameTimeInSeconds=GetGameTimeInSeconds()
    GravityOverrideParams=Get Gravity Override Params()
   )
   > Set GravityOverride(GravityOverride=@n3)
   n4 = CheckForBackbling()
   > Set WearingAtBack(WearingAtBack:bool=OR Boolean!(A=@n4, B=@n4.bIsBackblingACape, C:bool=@n4.bIsBackblingLong) -> (ReturnValue)) -> (Output_Get:bool)
   > Set isWearingBackpack(isWearingBackpack:bool=AND Boolean!(A=@n4, B=NOT Boolean(A=@n4.bIsBackblingACape), C:bool=NOT Boolean(A=@n4.bIsBackblingLong)) -> (ReturnValue)) -> (Output_Get:bool)
   > Set Is Wearing Cape(Is Wearing Cape=OR Boolean!(A=@n4.bIsBackblingACape, B=@n4.bIsBackblingLong) -> (ReturnValue))
   Event OnCharacterCustomizationLoaded()
   > n5 = Sequence!(execute) -> (then_0, then_1)
   ResetDynamics(execute=@n5.then_1, InTeleportType=TeleportPhysics)
   > ResetDynamics(InTeleportType=ResetPhysics)
   n6 = Cast To PlayerPawn_Athena(execute=@n5.then_0, Object=TryGetPawnOwner())
   > Set Body_Mesh(Body_Mesh=GetSkeletalMeshForPartType(self=@n6))
   Event BlueprintInitializeAnimation()
   > Set Skel Mesh(Skel Mesh=GetOwnerSkeletalMeshComponent())
   n7 = NOT Boolean(A=Get bIsOnGround() -> (bIsOnGround:bool))
   n8 = Get bIsTacticalSprinting()
   n9 = Branch(Condition=OR Boolean(A=AND Boolean!(A=@n8, B=@n7) -> (ReturnValue), B=Get bIsInFrontEnd()))
   Do Once(execute=@n9.else, Reset=@n9.then, Start Closed=true)
   > Delay(Duration=0.3)
   > ResetDynamics(InTeleportType=TeleportPhysics)
   > ResetDynamics(InTeleportType=ResetPhysics)
   Delay(Duration=0.1)
  ]}
]
variables [
 Body_Mesh {type=SkeletalMeshComponent}
 Gravity Override Params {type=GravityOverrideParamsStruct}
 GravityOverride {type=Vector}
 Is Wearing Cape {type=boolean}
 isWearingBackpack {type=boolean}
 Linear Scale Backpack {type=Vector}
 Linear Scale Cape {type=Vector}
 Linear Scale Free {type=Vector}
 Skel Mesh {type=SkeletalMeshComponent}
 WearingAtBack {type=boolean}
]
